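# WBP_UIKit_Throbber_Base is a base UI kit widget showing a loading throbber as a single material-driven Image, configurable by style, size and background.
# PreConstruct calls InitializeThrobber, which applies the style material and caches its dynamic instance for the size and background functions.
# Size selects the square brush dimensions and adjusts ring thickness parameters on that instance, while the background toggle drives a scalar parameter.
# Only the Monochromatic style has a material wired up.
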WidgetBlueprint /FortUILibrary/Systems/Throbber/WBP_UIKit_Throbber_Base
graphs [
 EventGraph {nodes [
   Event PreConstruct()
   > InitializeThrobber()
  ]}
 InitializeStyle {nodes [
   n1 = InitializeStyle!() -> (then, StyleType:E_UI_ThrobberType Enum=NewEnumerator0) {is=title K2Node_FunctionEntry}
   > Set ThrobberStyle(ThrobberStyle:E_UI_ThrobberType Enum=@n1) -> (Output_Get:E_UI_ThrobberType Enum=NewEnumerator0)
   > SetBrushFromMaterial(
    self=Get Throbber() -> (Throbber:Image Object Reference)
    Material=Select!(
     NewEnumerator0:Material Interface Object Reference=Get MonochromaticInstance() -> (MonochromaticInstance:Material Instance Object Reference)
     Index:E_UI_ThrobberType Enum=Get ThrobberStyle() -> (ThrobberStyle:E_UI_ThrobberType Enum=NewEnumerator0)
    ) -> (ReturnValue:Material Interface Object Reference)
   )
   > n2 = GetDynamicMaterial(self=Get Throbber())
   > Set CurrentMaterialInstance(CurrentMaterialInstance:Material Instance Dynamic Object Reference=@n2) -> (Output_Get:Material Instance Dynamic Object Reference)
  ]}
 InitializeThrobber {nodes [
   InitializeThrobber() {is=title K2Node_FunctionEntry}
   > InitializeStyle!(execute, self, StyleType:E_UI_ThrobberType Enum=Get ThrobberStyle() -> (ThrobberStyle=NewEnumerator0)) -> (then) {is=title K2Node_CallFunction}
   > SetThrobberSize!(execute, self, Size:E_UI_ThrobberSize Enum=Get ThrobberSize() -> (ThrobberSize:E_UI_ThrobberSize Enum=NewEnumerator0)) -> (then)
   > SetBackgroundVisibility!(execute, self, ShowBackground:bool=Get DisplayBackground() -> (DisplayBackground:bool)) -> (then)
  ]}
 SetBackgroundVisibility {nodes [
   n1 = SetBackgroundVisibility!() -> (then, ShowBackground:bool=~) {is=title K2Node_FunctionEntry}
   > Set DisplayBackground(DisplayBackground:bool=@n1) -> (Output_Get:bool)
   > SetScalarParameterValue(
    self=Get CurrentMaterialInstance() -> (CurrentMaterialInstance:Material Instance Dynamic Object Reference)
    ParameterName=Display Background
    Value=Select!(Option 0:float=0, Option 1:float=1, Index:bool=Get DisplayBackground()) -> (ReturnValue:float)
   )
  ]}
 SetThrobberSize {nodes [
   n1 = SetThrobberSize!() -> (then, Size:E_UI_ThrobberSize Enum=NewEnumerator0) {is=title K2Node_FunctionEntry}
   > Set ThrobberSize(ThrobberSize:E_UI_ThrobberSize Enum=@n1) -> (Output_Get:E_UI_ThrobberSize Enum=NewEnumerator0)
   n2 = "To Float (Integer)"(InInt=Get SizeSmall() -> (SizeSmall:int=0))
   n3 = "To Float (Integer)"(InInt=Get SizeMedium() -> (SizeMedium:int=0))
   n4 = "To Float (Integer)"(InInt=Get SizeLarge() -> (SizeLarge:int=0))
   n5 = Get Throbber()
   n6 = Get!(self:Image Object Reference=@n5) -> (
    Brush_TintColor:Slate Color Structure
    Brush_DrawAs:ESlateBrushDrawType Enum
    Brush_Tiling:ESlateBrushTileType Enum
    Brush_Mirroring:ESlateBrushMirrorType Enum
    Brush_ImageSize:Vector 2D Structure
    Brush_Margin:Margin Structure
    Brush_ResourceObject:Object Reference
    Brush_OutlineSettings:Slate Brush Outline Settings Structure
   )
   > SetBrush!(
    execute
    self=@n5
    InBrush_TintColor:Slate Color Structure=@n6
    InBrush_DrawAs:ESlateBrushDrawType Enum=@n6.Brush_DrawAs
    InBrush_Tiling:ESlateBrushTileType Enum=@n6.Brush_Tiling
    InBrush_Mirroring:ESlateBrushMirrorType Enum=@n6.Brush_Mirroring
    InBrush_ImageSize:Vector 2D Structure=Select!(
     NewEnumerator0_X:double=@n2
     NewEnumerator0_Y:double=@n2
     NewEnumerator1_X:double=@n3
     NewEnumerator1_Y:double=@n3
     NewEnumerator2_X:double=@n4
     NewEnumerator2_Y:double=@n4
     Index:E_UI_ThrobberSize Enum=Get ThrobberSize() -> (ThrobberSize=NewEnumerator0)
    ) -> (ReturnValue:Vector 2D Structure)
    InBrush_Margin:Margin Structure=@n6.Brush_Margin
    InBrush_ResourceObject:Object Reference=@n6.Brush_ResourceObject
    InBrush_OutlineSettings:Slate Brush Outline Settings Structure=@n6.Brush_OutlineSettings
   ) -> (then)
   n7 = Get CurrentMaterialInstance()
   > SetScalarParameterValue(
    self=@n7
    ParameterName=Background Thickness
    Value=Select!(NewEnumerator0:float=0.225, NewEnumerator1:float=0.2, NewEnumerator2:float=0.175, Index=Get ThrobberSize() -> (ThrobberSize=NewEnumerator0)) -> (ReturnValue:float)
   )
   > SetScalarParameterValue(
    self=@n7
    ParameterName=Inner Circle Thickness
    Value=Select(NewEnumerator0=0.175, NewEnumerator1=0.15, NewEnumerator2=0.125, Index=Get ThrobberSize() -> (ThrobberSize=NewEnumerator0))
   )
  ]}
]
variables [
 CurrentMaterialInstance {type=MaterialInstanceDynamic}
 DisplayBackground {type=boolean, default=true}
 MonochromaticInstance {type=MaterialInstance, default=/FortUILibrary/UIKit/Throbber/MI_UIKit_Monochrome_Throbber}
 SizeLarge {type=integer, default=152}
 SizeMedium {type=integer, default=120}
 SizeSmall {type=integer, default=48}
 ThrobberSize {type=E_UI_ThrobberSize, default=Small}
 ThrobberStyle {type=E_UI_ThrobberType, default=Monochromatic}
]
widgets {
 info {inheritedWidgetCount=0, parentClass=&UserWidget, rootWidgetClass=&Image, widgetCount=1}
 widgets [Throbber {
   widgetClassPath=&Image
   bIsVariable=true
   widget=Throbber
   properties {
    visibility=HitTestInvisible
    brush {
     bIsDynamicallyLoaded=false
     drawAs=Image
     imageType=NoImage
     mirroring=NoMirror
     resourceName=None
     resourceObject=&MI_UIKit_Monochrome_Throbber
     tiling=NoTile
     imageSize=<32 32>
     margin=<ltrb 0 0 0 0>
     outlineSettings {
      bUseBrushTransparency=false
      roundingType=HalfHeightRadius
      width=0
      color {colorUseRule=UseColor_Specified, specifiedColor=<rgba 0 0 0 0>}
      cornerRadii=<0 0 0 0>
     }
     tintColor {colorUseRule=UseColor_Specified, specifiedColor=<rgba 1 1 1 1>}
     uVRegion {bIsValid=false, max=<0 0>, min=<0 0>}
    }
   }
  }]
}
