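# Translucent UI material for a map marker shadow in the challenges UI, built from a soft radial gradient.
# UVs are multiplied by the Offset and Alpha parameter's RG, then fed to RadialGradientExponential, centered by RadialGradient Input's RG.
# That parameter's B drives both radius and density, leaving its A unused.
# The gradient lerps toward Offset and Alpha's B, alongside a separate dark navy RGB-1 color parameter.
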Material /Game/Athena/UI/Challenges/Art/MapIcons/M_UI_MapMarker_Shadow
graph {
 properties {
  bUsedWithStaticMesh=false
  blendMode=BLEND_Translucent
  materialDomain=MD_UI
  pixelDepthOffsetMode=PDOM_Legacy
  usedShadingModels=Default Lit
 }
 nodes [
  n1 = VectorParameter() {parameterName=Offset and Alpha, defaultValue=<rgba 0.85 1.8159 0.4 0.35>}
  n2 = TextureCoordinate()
  n3 = ComponentMask(None=@n1.RGB) {g=true, r=true}
  n4 = VectorParameter() {parameterName=RadialGradient Input, defaultValue=<rgba 0.4238 1.324 0.2674 2.0244>}
  LinearInterpolate(
   B=@n1.B
   Alpha=MaterialFunctionCall(UVs=Multiply(A=@n2, B=@n3), CenterPosition=ComponentMask(None=@n4.RGB) {g=true, r=true}, Radius=@n4.B, Density=@n4.B) {materialFunction=&RadialGradientExponential}.RadialGradientExponential
  )
  VectorParameter() {parameterName=RGB-1, defaultValue=<rgba 0 0.0162 0.1125 1>}
 ]
}
