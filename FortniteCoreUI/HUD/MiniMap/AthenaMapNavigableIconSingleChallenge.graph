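# Widget Blueprint for a single-challenge map pin, extending the native AthenaMapNavigableIconSingleChallenge class.
# OnChallengeSet either picks a complete, locked, default, tracked or stacked material and sets progress, or shows a pinned marker indexed through a material parameter.
# Other native events toggle bang, shadow and in-game state, and swap or restore the icon texture.
# Intro animations stagger by random delay; selection plays animation and sound, while hover events are empty.
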WidgetBlueprint /FortniteCoreUI/HUD/MiniMap/AthenaMapNavigableIconSingleChallenge
animations [
 OnSelectAnim {
  binding_count=2
  length_seconds=0.3
  object=&self:OnSelectAnim
  sequence {
   bindings [
    Image_Flare {
     id=#1
     possessed_class=Image
     tracks [
      {
       type=MovieSceneWidgetMaterialTrack
       sections [{type=MovieSceneParameterSection, channels [Content-Offset {type=MovieSceneScriptingFloatChannel, keys [[6,1],[9,0.2],[18,0.2]]}]}]
      }
      {
       type=MovieScene2DTransformTrack
       sections [{
         type=MovieScene2DTransformSection
         channels [
          Translation.X {type=MovieSceneScriptingFloatChannel}
          Translation.Y {type=MovieSceneScriptingFloatChannel}
          Angle {type=MovieSceneScriptingFloatChannel}
          Scale.X {type=MovieSceneScriptingFloatChannel, thinned_from=3, keys [[5.993,0.75],[18,1.75]]}
          Scale.Y {type=MovieSceneScriptingFloatChannel, thinned_from=3, keys [[5.993,0.75],[18,1.75]]}
          Shear.X {type=MovieSceneScriptingFloatChannel}
          Shear.Y {type=MovieSceneScriptingFloatChannel}
         ]
        }]
      }
      {
       type=MovieSceneFloatTrack
       display_name=Render Opacity
       sections [{type=MovieSceneFloatSection, channels [None {type=MovieSceneScriptingFloatChannel, keys [[0,0],[3,1],[6,1],[9,0.85],[18,0]]}]}]
      }
     ]
    }
    Overlay_0 {
     id=#2
     possessed_class=Overlay
     tracks [{
       type=MovieScene2DTransformTrack
       sections [{
         type=MovieScene2DTransformSection
         channels [
          Translation.X {type=MovieSceneScriptingFloatChannel}
          Translation.Y {type=MovieSceneScriptingFloatChannel}
          Angle {type=MovieSceneScriptingFloatChannel}
          Scale.X {type=MovieSceneScriptingFloatChannel, keys [[6,1],[15,1.2883],[18,1.25]]}
          Scale.Y {type=MovieSceneScriptingFloatChannel, keys [[6,1],[15,1.2883],[18,1.25]]}
          Shear.X {type=MovieSceneScriptingFloatChannel}
          Shear.Y {type=MovieSceneScriptingFloatChannel}
         ]
        }]
      }]
    }
   ]
   display_rate {numerator=60}
  }
 }
 OnStartAnim {
  binding_count=4
  length_seconds=0.3
  object=&self:OnStartAnim
  sequence {bindings [
    Image_Complete {
     id=#3
     possessed_class=Image
     tracks [
      {
       type=MovieSceneFloatTrack
       display_name=Render Opacity
       sections [{type=MovieSceneFloatSection, channels [None {type=MovieSceneScriptingFloatChannel, keys [[4,0],[6,1]]}]}]
      }
      {
       type=MovieScene2DTransformTrack
       sections [{
         type=MovieScene2DTransformSection
         channels [
          Translation.X {type=MovieSceneScriptingFloatChannel}
          Translation.Y {type=MovieSceneScriptingFloatChannel}
          Angle {type=MovieSceneScriptingFloatChannel}
          Scale.X {type=MovieSceneScriptingFloatChannel, keys [[4,0.5],[6,1]]}
          Scale.Y {type=MovieSceneScriptingFloatChannel, keys [[4,0.5],[6,1]]}
          Shear.X {type=MovieSceneScriptingFloatChannel}
          Shear.Y {type=MovieSceneScriptingFloatChannel}
         ]
        }]
      }
     ]
    }
    Image_Bang {
     id=#4
     possessed_class=Image
     tracks [
      {
       type=MovieSceneFloatTrack
       display_name=Render Opacity
       sections [{type=MovieSceneFloatSection, channels [None {type=MovieSceneScriptingFloatChannel, keys [[4,0],[6,1]]}]}]
      }
      {
       type=MovieScene2DTransformTrack
       sections [{
         type=MovieScene2DTransformSection
         channels [
          Translation.X {type=MovieSceneScriptingFloatChannel}
          Translation.Y {type=MovieSceneScriptingFloatChannel}
          Angle {type=MovieSceneScriptingFloatChannel}
          Scale.X {type=MovieSceneScriptingFloatChannel, keys [[4,0.5],[6,1]]}
          Scale.Y {type=MovieSceneScriptingFloatChannel, keys [[4,0.5],[6,1]]}
          Shear.X {type=MovieSceneScriptingFloatChannel}
          Shear.Y {type=MovieSceneScriptingFloatChannel}
         ]
        }]
      }
     ]
    }
    Overlay_0 {
     id=#5
     possessed_class=Overlay
     tracks [
      {
       type=MovieSceneFloatTrack
       display_name=Render Opacity
       sections [{type=MovieSceneFloatSection, channels [None {type=MovieSceneScriptingFloatChannel, keys [[0,0],[4,1]]}]}]
      }
      {
       type=MovieScene2DTransformTrack
       sections [{
         type=MovieScene2DTransformSection
         channels [
          Translation.X {type=MovieSceneScriptingFloatChannel}
          Translation.Y {type=MovieSceneScriptingFloatChannel}
          Angle {type=MovieSceneScriptingFloatChannel}
          Scale.X {type=MovieSceneScriptingFloatChannel, keys [[0,0.5],[4,1]]}
          Scale.Y {type=MovieSceneScriptingFloatChannel, keys [[0,0.5],[4,1]]}
          Shear.X {type=MovieSceneScriptingFloatChannel}
          Shear.Y {type=MovieSceneScriptingFloatChannel}
         ]
        }]
      }
     ]
    }
    Image_ShadowFor3dMap {
     id=#6
     possessed_class=Image
     tracks [{
       type=MovieSceneFloatTrack
       display_name=Render Opacity
       sections [{type=MovieSceneFloatSection, channels [None {type=MovieSceneScriptingFloatChannel, keys [[1,0],[4,1]]}]}]
      }]
    }
   ]}
 }
 Flare-Loop {
  binding_count=1
  length_seconds=1.25
  object=&self:Flare-Loop
  sequence {
   bindings [Image_Flare {
     id=#1
     possessed_class=Image
     tracks [
      {
       type=MovieSceneWidgetMaterialTrack
       sections [{
         type=MovieSceneParameterSection
         channels [Content-Offset {type=MovieSceneScriptingFloatChannel, thinned_from=4, keys [[6,1],[9,0.3],[75,0.3]]}]
        }]
      }
      {
       type=MovieScene2DTransformTrack
       sections [{
         type=MovieScene2DTransformSection
         channels [
          Translation.X {type=MovieSceneScriptingFloatChannel}
          Translation.Y {type=MovieSceneScriptingFloatChannel}
          Angle {type=MovieSceneScriptingFloatChannel}
          Scale.X {type=MovieSceneScriptingFloatChannel, keys [[5.993,0.75],[30,1.5]]}
          Scale.Y {type=MovieSceneScriptingFloatChannel, keys [[5.993,0.75],[30,1.5]]}
          Shear.X {type=MovieSceneScriptingFloatChannel}
          Shear.Y {type=MovieSceneScriptingFloatChannel}
         ]
        }]
      }
      {
       type=MovieSceneFloatTrack
       display_name=Render Opacity
       sections [{type=MovieSceneFloatSection, channels [None {type=MovieSceneScriptingFloatChannel, keys [[0,0],[3,1],[6,1],[9,0.85],[30,0]]}]}]
      }
     ]
    }]
   display_rate {numerator=60}
  }
 }
 OnHover {
  binding_count=1
  length_seconds=0.15
  object=&self:OnHover
  sequence {
   bindings [Image_Icon {
     id=#7
     possessed_class=Image
     tracks [{
       type=MovieSceneWidgetMaterialTrack
       sections [{
         type=MovieSceneParameterSection
         channels [
          Icon-State {type=MovieSceneScriptingFloatChannel, const [0,2]}
          Hovered {type=MovieSceneScriptingFloatChannel, const [0,2]}
          Selected {type=MovieSceneScriptingFloatChannel, keys [[0,0],[9,1]]}
         ]
        }]
      }]
    }]
   display_rate {numerator=60}
  }
 }
 OnSelectAnim-FromHover {
  binding_count=1
  length_seconds=0.2
  object=&self:OnSelectAnim-FromHover
  sequence {
   bindings [Overlay_0 {
     id=#2
     possessed_class=Overlay
     tracks [
      {
       type=MovieScene2DTransformTrack
       sections [{
         type=MovieScene2DTransformSection
         channels [
          Translation.X {type=MovieSceneScriptingFloatChannel}
          Translation.Y {type=MovieSceneScriptingFloatChannel}
          Angle {type=MovieSceneScriptingFloatChannel}
          Scale.X {type=MovieSceneScriptingFloatChannel, keys [[8,0.85],[18,1.125]]}
          Scale.Y {type=MovieSceneScriptingFloatChannel, keys [[8,0.85],[18,1.125]]}
          Shear.X {type=MovieSceneScriptingFloatChannel}
          Shear.Y {type=MovieSceneScriptingFloatChannel}
         ]
        }]
      }
      {
       type=MovieSceneFloatTrack
       display_name=Render Opacity
       sections [{type=MovieSceneFloatSection, channels [None {type=MovieSceneScriptingFloatChannel, keys [[6,0],[15,1]]}]}]
      }
     ]
    }]
   display_rate {numerator=60}
  }
 }
]
graphs [
 DisplayPinSelection {nodes [
   DisplayPinSelection!() -> (then, IsSelected:bool=~, Index:int=4) {is=title K2Node_FunctionEntry}
   > SetBrushFromMaterial(
    self=Get Image_Icon() -> (Image_Icon:Image Object Reference)
    Material=Select(
     Option 0:Material Interface Object Reference=/Game/Athena/UI/Challenges/Art/MapIcons/MI_UI_QuestPin_Deselected
     Option 1:Material Interface Object Reference=/Game/Athena/UI/Challenges/Art/MapIcons/MI_UI_QuestPin_Selected
     Index=Get IsSelected() -> (IsSelected:bool)
    ) -> (ReturnValue:Material Interface Object Reference)
   )
   > n1 = GetDynamicMaterial(self=Get Image_Icon())
   > SetScalarParameterValue(
    self=@n1
    ParameterName=Get Param_BackingTypeSelector() -> (Param_BackingTypeSelector:Name=None)
    Value="To Float (Integer)"(InInt=Get Index() -> (Index:int=0))
   )
   > SetVisibility(
    self=[
     Get Image_Flare() -> (Image_Flare:Fort Mobile Image Object Reference)
     Get Image_Bang() -> (Image_Bang:Fort Mobile Image Object Reference)
     Get Image_Complete() -> (Image_Complete:Fort Mobile Image Object Reference)
     Get Image_ShadowFor3dMap() -> (Image_ShadowFor3dMap:Image Object Reference)
    ]
    InVisibility=Collapsed
   )
  ]}
 EventGraph {nodes [
   n1 = Event OnChangeBangVisibility!() -> (OutputDelegate, then, bShouldShow:bool) {is=title K2Node_Event}
   > SetVisibility(
    self=Get Image_Bang()
    InVisibility=Select(Option 0:ESlateVisibility Enum=Collapsed, Option 1:ESlateVisibility Enum=SelfHitTestInvisible, Index=@n1.bShouldShow) -> (ReturnValue:ESlateVisibility Enum)
   )
   Event OnSelected!() -> (OutputDelegate, then, bUserInteraction:bool) {is=title K2Node_Event}
   > PlayAnimation(InAnimation=Get OnSelectAnim-FromHover() -> (OnSelectAnim-FromHover:Widget Animation Object Reference))
   > PlaySound2D(Sound=/Game/Sounds/Fort_UI_Sounds/S12_Frontend/UI_Table_HoverButton_01_Cue)
   Event PreConstruct()
   > SetSinglePinnedMarker!(execute, self, bSelected:bool=true) -> (then)
   n2 = Event OnSetIsInGame!() -> (OutputDelegate, then, bIsInGame:bool) {is=title K2Node_Event}
   > n3 = GetDynamicMaterial(self=Get Image_Icon())
   > SetScalarParameterValue(self=@n3, ParameterName=bIsIngame, Value=Select(Option 0:double=0, Option 1:double=1, Index=@n2.bIsInGame) -> (ReturnValue:double))
   > SetVisibility(self=Get SizeBox_13() -> (SizeBox_13:Size Box Object Reference), InVisibility=SelfHitTestInvisible)
   > StopAllAnimations(self:User Widget Object Reference)
   > SetVisibility(
    self=Get Image_ShadowFor3dMap()
    InVisibility=Select(Option 0:ESlateVisibility Enum=SelfHitTestInvisible, Option 1:ESlateVisibility Enum=Collapsed, Index=@n2.bIsInGame) -> (ReturnValue:ESlateVisibility Enum)
   )
   n4 = Event OnChallengeIconOverriden!() -> (OutputDelegate, then, Icon:Texture 2D Object Reference) {is=title K2Node_Event}
   > n5 = GetDynamicMaterial(self=Get Image_Icon())
   > n6 = Get Texture Parameter Value!(execute, self:Material Instance Dynamic Object Reference=@n5, ParameterName:Name=Icon) -> (then, ReturnValue:Texture Object Reference) {is=title K2Node_CallFunction}
   > Set DefaultIcon(DefaultIcon:Texture Object Reference=@n6) -> (Output_Get:Texture Object Reference)
   > SetTextureParameterValue(self=@n5, ParameterName=Icon, Value=@n4.Icon)
   n7 = Event OnChallengeSet!() -> (
    OutputDelegate
    then
    bIsComplete:bool
    bIsLocked:bool
    bHasIconOverride:bool
    Progress:float
    bShouldShowBang:bool
    bIsPinned:bool
    bChallengeSelected:bool
    bIsTracked:bool
    PinnedQuestIndex:int=0
   ) {is=title K2Node_Event}
   > SetVisibility(self=Get Image_Bang(), InVisibility=Select(Option 0=Collapsed, Option 1=SelfHitTestInvisible, Index=@n7.bShouldShowBang))
   > n8 = Branch(Condition="integer < integer"(A=@n7.PinnedQuestIndex))
   > SetOpacity!(execute, self:Image Object Reference=Get Image_Icon(), InOpacity:float=1) -> (then)
   > SetMaterialState!(
    execute
    self
    Complete:bool=@n7.bIsComplete
    Locked:bool=@n7.bIsLocked
    HasIconOverride:bool=@n7.bHasIconOverride
    IsTracked:bool=@n7.bIsTracked
   ) -> (then)
   > n9 = GetDynamicMaterial(self=Get Image_Icon())
   > SetScalarParameterValue(self=@n9, ParameterName=Progress, Value=@n7.Progress)
   SetMultiPinnedMarker!(execute=@n8.else, self, bSelected:bool=@n7.bChallengeSelected, PinnedIndex:int=@n7.PinnedQuestIndex) -> (then)
   Event OnChallengeReset() {is=title K2Node_Event}
   > n10 = GetDynamicMaterial(self=Get Image_Icon())
   > SetTextureParameterValue(self=@n10, ParameterName=Icon, Value=Get DefaultIcon() -> (DefaultIcon:Texture Object Reference))
   > SetScalarParameterValue(self=@n10, ParameterName=bIsLocked, Value=0)
   > SetScalarParameterValue(self=@n10, ParameterName=Progress, Value=0)
   > SetVisibility(self=Get Image_Complete(), InVisibility=Collapsed)
   Event OnBeginIntroAnimation() {is=title K2Node_Event}
   > SetVisibility(self=Get SizeBox_13(), InVisibility=Hidden)
   > Delay(Duration=RandomFloatInRange(Min=0.1, Max=0.6))
   > SetVisibility(self=Get SizeBox_13(), InVisibility=SelfHitTestInvisible)
   > PlayAnimation(InAnimation=Get OnStartAnim() -> (OnStartAnim:Widget Animation Object Reference))
   SetVisibility(
    self=Get Image_ShadowFor3dMap()
    InVisibility=Select(
     Option 0=Collapsed
     Option 1=SelfHitTestInvisible
     Index=Event OnSetHasLocation!() -> (OutputDelegate, then, bHasLocation:bool) {is=title K2Node_Event}.bHasLocation
    )
   )
   Event Construct()
   Event OnUnselected!() -> (OutputDelegate, then, bUserInteraction:bool) {is=title K2Node_Event}
   Event OnHovered!() -> (OutputDelegate, then, bUserInteraction:bool) {is=title K2Node_Event}
   Event OnUnhovered!() -> (OutputDelegate, then, bUserInteraction:bool) {is=title K2Node_Event}
  ]}
 SetMaterialState {nodes [
   n1 = SetMaterialState!() -> (then, Complete:bool=~, Locked:bool=~, HasIconOverride:bool=~, IsTracked:bool=~) {is=title K2Node_FunctionEntry}
   > n2 = Branch(Condition=@n1)
   > SetBrushFromMaterial(self=Get Image_Icon(), Material=/Game/Athena/UI/Challenges/Art/MapIcons/MI_UI_QuestPin_Complete)
   n3 = Branch(execute=@n2.else, Condition=@n1.Locked)
   > SetBrushFromMaterial(self=Get Image_Icon(), Material=/Game/Athena/UI/Challenges/Art/MapIcons/MI_UI_QuestPin_Locked)
   n4 = Branch(execute=@n3.else, Condition=@n1.HasIconOverride)
   > SetBrushFromMaterial(
    self=Get Image_Icon()
    Material=Select(
     Option 0:Material Interface Object Reference=/Game/Athena/UI/Challenges/Art/MapIcons/MI_UI_QuestPin_Default
     Option 1:Material Interface Object Reference=&MI_UI_QuestPin_Selected
     Index=Get IsTracked() -> (IsTracked:bool)
    ) -> (ReturnValue:Material Interface Object Reference)
   )
   SetBrushFromMaterial(execute=@n4.else, self=Get Image_Icon(), Material=/Game/Athena/UI/Challenges/Art/MapIcons/MI_UI_QuestPin_Stacked)
  ]}
 SetMultiPinnedMarker {nodes [
   SetMultiPinnedMarker!() -> (then, bSelected:bool=~, PinnedIndex:int) {is=title K2Node_FunctionEntry}
   > DisplayPinSelection!(execute, self, IsSelected:bool=Get bSelected() -> (bSelected:bool), Index:int=int + int(A=Get PinnedIndex() -> (PinnedIndex:int=0))) -> (then) {is=title K2Node_CallFunction}
  ]}
 SetSinglePinnedMarker {nodes [
   n1 = SetSinglePinnedMarker!() -> (then, bSelected:bool=~) {is=title K2Node_FunctionEntry}
   > DisplayPinSelection(IsSelected=@n1, Index=4)
  ]}
]
variables [
 DefaultIcon {type=Texture, default=null}
 Param_BackingTypeSelector {type=string, default=BackingTypeSelector}
]
widgets {
 info {
  inheritedWidgetCount=1
  parentClass=/Script/FortniteUI.AthenaMapNavigableIconSingleChallenge
  rootWidgetClass=&SizeBox
  widgetCount=8
 }
 widgets [
  SizeBox_13 {widgetClassPath=&SizeBox, bIsVariable=true, widget=SizeBox_13}
  Overlay_1 {
   widgetClassPath=&Overlay
   parent=SizeBox_13
   widget=Overlay_1
   slot {class=SizeBoxSlot, properties {horizontalAlignment=HAlign_Center}}
  }
  Image_ShadowFor3dMap {
   widgetClassPath=&Image
   bIsVariable=true
   parent=Overlay_1
   widget=Image_ShadowFor3dMap
   properties {
    visibility=Collapsed
    brush {
     bIsDynamicallyLoaded=false
     drawAs=Image
     imageType=NoImage
     mirroring=NoMirror
     resourceName=None
     resourceObject=/Game/Athena/UI/Challenges/Art/MapIcons/M_UI_MapMarker_Shadow
     tiling=NoTile
     imageSize=<64 64>
     margin=<ltrb 0 0 0 0>
     outlineSettings {
      bUseBrushTransparency=false
      roundingType=HalfHeightRadius
      width=0
      color {colorUseRule=UseColor_Specified, specifiedColor=<rgba 0 0 0 0>}
      cornerRadii=<0 0 0 0>
     }
     tintColor {colorUseRule=UseColor_Specified, specifiedColor=<rgba 1 1 1 1>}
     uVRegion {bIsValid=false, max=<0 0>, min=<0 0>}
    }
    renderTransform {angle=0, scale=<2 2>, shear=<0 0>, translation=<0 0>}
   }
   slot {class=OverlaySlot, properties {horizontalAlignment=HAlign_Center, verticalAlignment=VAlign_Center}}
  }
  Overlay_0 {
   widgetClassPath=&Overlay
   parent=Overlay_1
   widget=Overlay_0
   properties {renderTransformPivot=<0.5 1>}
   slot {class=OverlaySlot, properties {horizontalAlignment=HAlign_Center}}
  }
  Image_Flare {
   widgetClassPath=/Script/FortniteUI.FortMobileImage
   bIsVariable=true
   parent=Overlay_0
   widget=Image_Flare
   properties {
    mobileSizeMultiplier=1.5
    renderOpacity=0
    visibility=Collapsed
    brush {
     bIsDynamicallyLoaded=false
     drawAs=Image
     imageType=NoImage
     mirroring=NoMirror
     resourceName=None
     resourceObject=/Game/Athena/UI/Challenges/Art/MapIcons/M_UI_MapMarker_Glow
     tiling=NoTile
     imageSize=<64 64>
     margin=<ltrb 0 0 0 0>
     outlineSettings {
      bUseBrushTransparency=false
      roundingType=HalfHeightRadius
      width=0
      color {colorUseRule=UseColor_Specified, specifiedColor=<rgba 0 0 0 0>}
      cornerRadii=<0 0 0 0>
     }
     tintColor {colorUseRule=UseColor_Specified, specifiedColor=<rgba 1 1 1 1>}
     uVRegion {bIsValid=false, max=<0 0>, min=<0 0>}
    }
   }
   slot {class=OverlaySlot, properties {horizontalAlignment=HAlign_Center, verticalAlignment=VAlign_Center}}
  }
  Image_Icon {
   widgetClassPath=&FortMobileImage
   bInherited=true
   bIsVariable=true
   parent=Overlay_0
   widget=Image_Icon
   properties {
    mobileSizeMultiplier=1.5
    visibility=SelfHitTestInvisible
    brush {
     bIsDynamicallyLoaded=false
     drawAs=Image
     imageType=FullColor
     mirroring=NoMirror
     resourceName=None
     resourceObject=null
     tiling=NoTile
     imageSize=<48 48>
     margin=<ltrb 0 0 0 0>
     outlineSettings {
      bUseBrushTransparency=false
      roundingType=HalfHeightRadius
      width=0
      color {colorUseRule=UseColor_Specified, specifiedColor=<rgba 0 0 0 0>}
      cornerRadii=<0 0 0 0>
     }
     tintColor {colorUseRule=UseColor_Specified, specifiedColor=<rgba 1 1 1 1>}
     uVRegion {bIsValid=false, max=<0 0>, min=<0 0>}
    }
    renderTransformPivot=<0.5 1>
   }
   slot {class=OverlaySlot}
  }
  Image_Bang {
   widgetClassPath=&FortMobileImage
   bIsVariable=true
   parent=Overlay_0
   widget=Image_Bang
   properties {
    mobileSizeMultiplier=1.5
    visibility=Collapsed
    brush {
     bIsDynamicallyLoaded=false
     drawAs=Image
     imageType=NoImage
     mirroring=NoMirror
     resourceName=None
     resourceObject=/Game/Athena/UI/Challenges/Art/MapIcons/M_UI_MarkerState-Bang
     tiling=NoTile
     imageSize=<24 24>
     margin=<ltrb 0 0 0 0>
     outlineSettings {
      bUseBrushTransparency=false
      roundingType=HalfHeightRadius
      width=0
      color {colorUseRule=UseColor_Specified, specifiedColor=<rgba 0 0 0 0>}
      cornerRadii=<0 0 0 0>
     }
     tintColor {colorUseRule=UseColor_Specified, specifiedColor=<rgba 1 1 1 1>}
     uVRegion {bIsValid=false, max=<0 0>, min=<0 0>}
    }
    renderTransform {angle=0, scale=<1 1>, shear=<0 0>, translation=<4 -4>}
    renderTransformPivot=<0 1>
   }
   slot {class=OverlaySlot, properties {horizontalAlignment=HAlign_Right}}
  }
  Image_Complete {
   widgetClassPath=&FortMobileImage
   bIsVariable=true
   parent=Overlay_0
   widget=Image_Complete
   properties {
    mobileSizeMultiplier=1.5
    visibility=Collapsed
    brush {
     bIsDynamicallyLoaded=false
     drawAs=Image
     imageType=NoImage
     mirroring=NoMirror
     resourceName=None
     resourceObject=/Game/Athena/UI/Challenges/Art/MapIcons/M_UI_MarkerState-Complete
     tiling=NoTile
     imageSize=<24 24>
     margin=<ltrb 0 0 0 0>
     outlineSettings {
      bUseBrushTransparency=false
      roundingType=HalfHeightRadius
      width=0
      color {colorUseRule=UseColor_Specified, specifiedColor=<rgba 0 0 0 0>}
      cornerRadii=<0 0 0 0>
     }
     tintColor {colorUseRule=UseColor_Specified, specifiedColor=<rgba 1 1 1 1>}
     uVRegion {bIsValid=false, max=<0 0>, min=<0 0>}
    }
    renderTransform {angle=0, scale=<1 1>, shear=<0 0>, translation=<4 -4>}
    renderTransformPivot=<0 1>
   }
   slot {class=OverlaySlot, properties {horizontalAlignment=HAlign_Right}}
  }
 ]
}
