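# UI sound cue for hovering over the challenge table in the frontend, routed to the Fort_UI sound class at reduced volume.
# A single wave player plays the S13 challenge table hover wave through a modulator that adds slight random pitch variation.
# A concurrency override allows one instance at a time, stopping the oldest so rapid hovers do not stack.
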SoundCue /Game/Sounds/Fort_UI_Sounds/S12_Frontend/UI_Challenge_Table_Hover_01_Cue
root=Modulator_0
nodes [
 {id=Modulator_0, children [WavePlayer_0], properties {pitchMax=1.01, pitchMin=0.99}}
 {id=WavePlayer_0, properties {soundWaveAssetPtr=/Game/Sounds/Fort_UI_Sounds/S13_Frontend/S13_Challenge_Table_Hover}}
]
properties {
 bOverrideConcurrency=true
 maxDistance=2097152
 soundClassObject=&Fort_UI
 volumeMultiplier=0.5
 concurrencyOverrides {maxCount=1, resolutionRule=StopOldest}
}
