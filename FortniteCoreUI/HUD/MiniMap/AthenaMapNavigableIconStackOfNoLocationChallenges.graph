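# Map icon widget for a stack of challenges with no map location, showing progress, bang, remaining count and completed-of-total text.
# Native parent events swap the "Anywhere" marker material for complete, locked or icon-override states, then set progress and icon on its dynamic material.
# Hover and selection animate material parameters, scale and a looping flare, with sounds.
# Non-user selection jumps to the animation's end; in-game state stops animations and hides the shadow.
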WidgetBlueprint /FortniteCoreUI/HUD/MiniMap/AthenaMapNavigableIconStackOfNoLocationChallenges
animations [
 OnSelectedAnim {
  binding_count=5
  length_seconds=0.3
  object=&self:OnSelectedAnim
  sequence {bindings [
    Image_Flare {
     id=#1
     possessed_class=Image
     tracks [
      {
       type=MovieSceneFloatTrack
       display_name=Render Opacity
       sections [{type=MovieSceneFloatSection, channels [None {type=MovieSceneScriptingFloatChannel, keys [[0,0],[1,1],[2,1],[3,0.85],[6,0]]}]}]
      }
      {
       type=MovieScene2DTransformTrack
       sections [{
         type=MovieScene2DTransformSection
         channels [
          Translation.X {type=MovieSceneScriptingFloatChannel}
          Translation.Y {type=MovieSceneScriptingFloatChannel}
          Angle {type=MovieSceneScriptingFloatChannel}
          Scale.X {type=MovieSceneScriptingFloatChannel, keys [[1.9977,0.75],[2,0.825],[5,1.5],[6,1.75]]}
          Scale.Y {type=MovieSceneScriptingFloatChannel, keys [[1.9977,0.75],[2,0.825],[5,1.5],[6,1.75]]}
          Shear.X {type=MovieSceneScriptingFloatChannel}
          Shear.Y {type=MovieSceneScriptingFloatChannel}
         ]
        }]
      }
      {
       type=MovieSceneWidgetMaterialTrack
       sections [{type=MovieSceneParameterSection, channels [Content-Offset {type=MovieSceneScriptingFloatChannel, keys [[2,1],[3,0.2],[6,0.2]]}]}]
      }
     ]
    }
    Overlay_Internal {
     id=#2
     possessed_class=Overlay
     tracks [{
       type=MovieScene2DTransformTrack
       sections [{
         type=MovieScene2DTransformSection
         channels [
          Translation.X {type=MovieSceneScriptingFloatChannel}
          Translation.Y {type=MovieSceneScriptingFloatChannel}
          Angle {type=MovieSceneScriptingFloatChannel}
          Scale.X {type=MovieSceneScriptingFloatChannel, keys [[2,1],[5,1.28],[6,1.25]]}
          Scale.Y {type=MovieSceneScriptingFloatChannel, keys [[2,1],[5,1.28],[6,1.25]]}
          Shear.X {type=MovieSceneScriptingFloatChannel}
          Shear.Y {type=MovieSceneScriptingFloatChannel}
         ]
        }]
      }]
    }
    Image_Icon {
     id=#3
     possessed_class=Image
     tracks [{
       type=MovieSceneWidgetMaterialTrack
       sections [{
         type=MovieSceneParameterSection
         channels [
          Selected {type=MovieSceneScriptingFloatChannel, keys [[2,0],[5,1]]}
          Bot-Radius-S1 {type=MovieSceneScriptingFloatChannel, keys [[2,1.025],[5,1.04]]}
          Bot-Radius-S2 {type=MovieSceneScriptingFloatChannel, keys [[2,0.985],[4,0.9775],[5,1.005]]}
          Inner-Value-S1 {type=MovieSceneScriptingFloatChannel, const [0.25,2]}
          Inner-Value-S2 {type=MovieSceneScriptingFloatChannel, const [0,3]}
          Left-Inner-OffSet-S1 {type=MovieSceneScriptingFloatChannel, keys [[2,0.085],[5,0.065]]}
          Left-Inner-OffSet-S2 {type=MovieSceneScriptingFloatChannel, keys [[2,0.125],[4,0.155],[5,0.11]]}
          Right-Inner-OffSet-S1 {type=MovieSceneScriptingFloatChannel, keys [[2,0.115],[5,0.09]]}
          Right-Inner-OffSet-S2 {type=MovieSceneScriptingFloatChannel, keys [[2,0.155],[4,0.185],[5,0.14]]}
          Top-Inner-OffSet-S1 {type=MovieSceneScriptingFloatChannel, keys [[2,0.085],[5,0.065]]}
          Top-Inner-OffSet-S2 {type=MovieSceneScriptingFloatChannel, keys [[2,0.125],[4,0.155],[5,0.11]]}
          Icon-State {type=MovieSceneScriptingFloatChannel, keys [[2,0],[5,1]]}
         ]
        }]
      }]
    }
    CommonText_Stack_Count {
     id=#4
     possessed_class=CommonTextBlock
     tracks [
      {
       type=MovieScene2DTransformTrack
       sections [{
         type=MovieScene2DTransformSection
         channels [
          Translation.X {type=MovieSceneScriptingFloatChannel}
          Translation.Y {type=MovieSceneScriptingFloatChannel}
          Angle {type=MovieSceneScriptingFloatChannel}
          Scale.X {type=MovieSceneScriptingFloatChannel, keys [[2,1],[6,0.95]]}
          Scale.Y {type=MovieSceneScriptingFloatChannel, keys [[2,1],[6,0.95]]}
          Shear.X {type=MovieSceneScriptingFloatChannel}
          Shear.Y {type=MovieSceneScriptingFloatChannel}
         ]
        }]
      }
      {
       type=MovieSceneColorTrack
       display_name=Color and Opacity
       sections [{
         type=MovieSceneColorSection
         channels [
          Color.R {type=MovieSceneScriptingFloatChannel, keys [[2,1],[6,0.0024]]}
          Color.G {type=MovieSceneScriptingFloatChannel, keys [[2,1],[6,0.0395]]}
          Color.B {type=MovieSceneScriptingFloatChannel, keys [[2,1],[6,0.1329]]}
          Color.A {type=MovieSceneScriptingFloatChannel, const [1,2]}
         ]
        }]
      }
     ]
    }
    Image_ShadowFor3dMap {
     id=#5
     possessed_class=Image
     tracks [
      {
       type=MovieSceneWidgetMaterialTrack
       sections [{
         type=MovieSceneParameterSection
         channels [
          Offset and AlphaR {type=MovieSceneScriptingFloatChannel, const [0.85,2]}
          Offset and AlphaG {type=MovieSceneScriptingFloatChannel, const [1.6,2]}
          Offset and AlphaB {type=MovieSceneScriptingFloatChannel, keys [[1,0.3],[5,0.4]]}
          Offset and AlphaA {type=MovieSceneScriptingFloatChannel, const [0.35,2]}
         ]
        }]
      }
      {
       type=MovieScene2DTransformTrack
       sections [{
         type=MovieScene2DTransformSection
         channels [
          Translation.X {type=MovieSceneScriptingFloatChannel}
          Translation.Y {type=MovieSceneScriptingFloatChannel}
          Angle {type=MovieSceneScriptingFloatChannel}
          Scale.X {type=MovieSceneScriptingFloatChannel}
          Scale.Y {type=MovieSceneScriptingFloatChannel}
          Shear.X {type=MovieSceneScriptingFloatChannel}
          Shear.Y {type=MovieSceneScriptingFloatChannel}
         ]
        }]
      }
     ]
    }
   ]}
 }
 OnStartAnim {
  binding_count=6
  length_seconds=0.3
  object=&self:OnStartAnim
  sequence {bindings [
    Image_ShadowFor3dMap {
     id=#6
     possessed_class=Image
     tracks [{
       type=MovieSceneFloatTrack
       display_name=Render Opacity
       sections [{type=MovieSceneFloatSection, channels [None {type=MovieSceneScriptingFloatChannel, keys [[2,0],[4,1]]}]}]
      }]
    }
    Border_0 {
     id=#7
     possessed_class=Border
     tracks [
      {
       type=MovieSceneFloatTrack
       display_name=Render Opacity
       sections [{type=MovieSceneFloatSection, channels [None {type=MovieSceneScriptingFloatChannel, keys [[2,0],[4,1]]}]}]
      }
      {
       type=MovieScene2DTransformTrack
       sections [{
         type=MovieScene2DTransformSection
         channels [
          Scale.X {type=MovieSceneScriptingFloatChannel, keys [[2,0.5],[4,1]]}
          Scale.Y {type=MovieSceneScriptingFloatChannel, keys [[2,0.5],[4,1]]}
         ]
        }]
      }
     ]
    }
    Text_AllChallenges {
     id=#8
     possessed_class=CommonTextBlock
     tracks [{
       type=MovieSceneFloatTrack
       display_name=Render Opacity
       sections [{type=MovieSceneFloatSection, channels [None {type=MovieSceneScriptingFloatChannel, keys [[2,0],[4,1]]}]}]
      }]
    }
    CompletedCountText {
     id=#9
     possessed_class=CommonTextBlock
     tracks [{
       type=MovieSceneFloatTrack
       display_name=Render Opacity
       sections [{type=MovieSceneFloatSection, channels [None {type=MovieSceneScriptingFloatChannel, keys [[4,0],[6,1]]}]}]
      }]
    }
    Image_Bang {
     id=#10
     possessed_class=Image
     tracks [
      {
       type=MovieSceneFloatTrack
       display_name=Render Opacity
       sections [{type=MovieSceneFloatSection, channels [None {type=MovieSceneScriptingFloatChannel, keys [[4,0],[6,1]]}]}]
      }
      {
       type=MovieScene2DTransformTrack
       sections [{
         type=MovieScene2DTransformSection
         channels [
          Translation.X {type=MovieSceneScriptingFloatChannel}
          Translation.Y {type=MovieSceneScriptingFloatChannel}
          Angle {type=MovieSceneScriptingFloatChannel}
          Scale.X {type=MovieSceneScriptingFloatChannel, keys [[4,0.5],[6,1]]}
          Scale.Y {type=MovieSceneScriptingFloatChannel, keys [[4,0.5],[6,1]]}
          Shear.X {type=MovieSceneScriptingFloatChannel}
          Shear.Y {type=MovieSceneScriptingFloatChannel}
         ]
        }]
      }
     ]
    }
    Overlay_Internal {
     id=#11
     possessed_class=Overlay
     tracks [
      {
       type=MovieScene2DTransformTrack
       sections [{
         type=MovieScene2DTransformSection
         channels [
          Translation.X {type=MovieSceneScriptingFloatChannel}
          Translation.Y {type=MovieSceneScriptingFloatChannel}
          Angle {type=MovieSceneScriptingFloatChannel}
          Scale.X {type=MovieSceneScriptingFloatChannel, keys [[0,0.5],[4,1]]}
          Scale.Y {type=MovieSceneScriptingFloatChannel, keys [[0,0.5],[4,1]]}
          Shear.X {type=MovieSceneScriptingFloatChannel}
          Shear.Y {type=MovieSceneScriptingFloatChannel}
         ]
        }]
      }
      {
       type=MovieSceneFloatTrack
       display_name=Render Opacity
       sections [{type=MovieSceneFloatSection, channels [None {type=MovieSceneScriptingFloatChannel, keys [[0,0],[4,1]]}]}]
      }
     ]
    }
   ]}
 }
 Flare-Loop {
  binding_count=1
  length_seconds=1.25
  object=&self:Flare-Loop
  sequence {bindings [Image_Flare {
     id=#12
     possessed_class=Image
     tracks [
      {
       type=MovieSceneFloatTrack
       display_name=Render Opacity
       sections [{type=MovieSceneFloatSection, channels [None {type=MovieSceneScriptingFloatChannel, keys [[0,0],[1,1],[2,1],[3,0.85],[10,0]]}]}]
      }
      {
       type=MovieScene2DTransformTrack
       sections [{
         type=MovieScene2DTransformSection
         channels [
          Translation.X {type=MovieSceneScriptingFloatChannel}
          Translation.Y {type=MovieSceneScriptingFloatChannel}
          Angle {type=MovieSceneScriptingFloatChannel}
          Scale.X {type=MovieSceneScriptingFloatChannel, keys [[1.9977,0.75],[10,1.5]]}
          Scale.Y {type=MovieSceneScriptingFloatChannel, keys [[1.9977,0.75],[10,1.5]]}
          Shear.X {type=MovieSceneScriptingFloatChannel}
          Shear.Y {type=MovieSceneScriptingFloatChannel}
         ]
        }]
      }
      {
       type=MovieSceneWidgetMaterialTrack
       sections [{
         type=MovieSceneParameterSection
         channels [Content-Offset {type=MovieSceneScriptingFloatChannel, thinned_from=4, keys [[2,1],[3,0.3],[25,0.3]]}]
        }]
      }
     ]
    }]}
 }
 OnHover {
  binding_count=2
  length_seconds=0.2
  object=&self:OnHover
  sequence {bindings [
    Image_Icon {
     id=#13
     possessed_class=FortMobileImage
     tracks [{
       type=MovieSceneWidgetMaterialTrack
       sections [{
         type=MovieSceneParameterSection
         channels [
          Hovered {type=MovieSceneScriptingFloatChannel, keys [[0,0],[3,1]]}
          Selected {type=MovieSceneScriptingFloatChannel, keys [[0,0],[3,1]]}
          Icon-State {type=MovieSceneScriptingFloatChannel, keys [[0,0],[3,1]]}
         ]
        }]
      }]
    }
    CommonText_Stack_Count {
     id=#14
     possessed_class=CommonTextBlock
     tracks [{
       type=MovieSceneColorTrack
       display_name=Color and Opacity
       sections [{
         type=MovieSceneColorSection
         channels [
          Color.R {type=MovieSceneScriptingFloatChannel, keys [[0,1],[4,0.0024]]}
          Color.G {type=MovieSceneScriptingFloatChannel, keys [[0,1],[4,0.0395]]}
          Color.B {type=MovieSceneScriptingFloatChannel, keys [[0,1],[4,0.1329]]}
          Color.A {type=MovieSceneScriptingFloatChannel, const [1,2]}
         ]
        }]
      }]
    }
   ]}
 }
 OnSelectAnim-FromHover {
  binding_count=5
  length_seconds=0.3
  object=&self:OnSelectAnim-FromHover
  sequence {bindings [
    Image_Flare {
     id=#1
     possessed_class=Image
     tracks [
      {
       type=MovieSceneFloatTrack
       display_name=Render Opacity
       sections [{type=MovieSceneFloatSection, channels [None {type=MovieSceneScriptingFloatChannel, keys [[0,0],[1,1],[2,1],[3,0.85],[6,0]]}]}]
      }
      {
       type=MovieScene2DTransformTrack
       sections [{
         type=MovieScene2DTransformSection
         channels [
          Translation.X {type=MovieSceneScriptingFloatChannel}
          Translation.Y {type=MovieSceneScriptingFloatChannel}
          Angle {type=MovieSceneScriptingFloatChannel}
          Scale.X {type=MovieSceneScriptingFloatChannel, keys [[1.9977,0.75],[2,0.825],[5,1.5],[6,1.75]]}
          Scale.Y {type=MovieSceneScriptingFloatChannel, keys [[1.9977,0.75],[2,0.825],[5,1.5],[6,1.75]]}
          Shear.X {type=MovieSceneScriptingFloatChannel}
          Shear.Y {type=MovieSceneScriptingFloatChannel}
         ]
        }]
      }
      {
       type=MovieSceneWidgetMaterialTrack
       sections [{type=MovieSceneParameterSection, channels [Content-Offset {type=MovieSceneScriptingFloatChannel, keys [[2,1],[3,0.2],[6,0.2]]}]}]
      }
     ]
    }
    Overlay_Internal {
     id=#2
     possessed_class=Overlay
     tracks [{
       type=MovieScene2DTransformTrack
       sections [{
         type=MovieScene2DTransformSection
         channels [
          Translation.X {type=MovieSceneScriptingFloatChannel}
          Translation.Y {type=MovieSceneScriptingFloatChannel}
          Angle {type=MovieSceneScriptingFloatChannel}
          Scale.X {type=MovieSceneScriptingFloatChannel, keys [[2,1],[5,1.28],[6,1.25]]}
          Scale.Y {type=MovieSceneScriptingFloatChannel, keys [[2,1],[5,1.28],[6,1.25]]}
          Shear.X {type=MovieSceneScriptingFloatChannel}
          Shear.Y {type=MovieSceneScriptingFloatChannel}
         ]
        }]
      }]
    }
    Image_Icon {
     id=#3
     possessed_class=Image
     tracks [{
       type=MovieSceneWidgetMaterialTrack
       sections [{
         type=MovieSceneParameterSection
         channels [
          Selected {type=MovieSceneScriptingFloatChannel, const [1,3]}
          Hovered {type=MovieSceneScriptingFloatChannel, keys [[0,1],[2,0]]}
          Icon-State {type=MovieSceneScriptingFloatChannel, const [1,2]}
          Top-Inner-OffSet-S2 {type=MovieSceneScriptingFloatChannel, keys [[2,0.125],[4,0.155],[5,0.11]]}
          Top-Inner-OffSet-S1 {type=MovieSceneScriptingFloatChannel, keys [[2,0.085],[5,0.065]]}
          Right-Inner-OffSet-S2 {type=MovieSceneScriptingFloatChannel, keys [[2,0.155],[4,0.185],[5,0.14]]}
          Right-Inner-OffSet-S1 {type=MovieSceneScriptingFloatChannel, keys [[2,0.115],[5,0.09]]}
          Left-Inner-OffSet-S2 {type=MovieSceneScriptingFloatChannel, keys [[2,0.125],[4,0.155],[5,0.11]]}
          Left-Inner-OffSet-S1 {type=MovieSceneScriptingFloatChannel, keys [[2,0.085],[5,0.065]]}
          Inner-Value-S2 {type=MovieSceneScriptingFloatChannel, const [0,3]}
          Inner-Value-S1 {type=MovieSceneScriptingFloatChannel, const [0.25,2]}
          Bot-Radius-S2 {type=MovieSceneScriptingFloatChannel, keys [[2,0.985],[4,0.9775],[5,1.005]]}
          Bot-Radius-S1 {type=MovieSceneScriptingFloatChannel, keys [[2,1.025],[5,1.04]]}
         ]
        }]
      }]
    }
    CommonText_Stack_Count {
     id=#4
     possessed_class=CommonTextBlock
     tracks [
      {
       type=MovieScene2DTransformTrack
       sections [{
         type=MovieScene2DTransformSection
         channels [
          Translation.X {type=MovieSceneScriptingFloatChannel}
          Translation.Y {type=MovieSceneScriptingFloatChannel}
          Angle {type=MovieSceneScriptingFloatChannel}
          Scale.X {type=MovieSceneScriptingFloatChannel, keys [[2,1],[6,0.95]]}
          Scale.Y {type=MovieSceneScriptingFloatChannel, keys [[2,1],[6,0.95]]}
          Shear.X {type=MovieSceneScriptingFloatChannel}
          Shear.Y {type=MovieSceneScriptingFloatChannel}
         ]
        }]
      }
      {
       type=MovieSceneColorTrack
       display_name=Color and Opacity
       sections [{
         type=MovieSceneColorSection
         channels [
          Color.R {type=MovieSceneScriptingFloatChannel, const [1,1]}
          Color.G {type=MovieSceneScriptingFloatChannel, const [1,1]}
          Color.B {type=MovieSceneScriptingFloatChannel, const [1,1]}
          Color.A {type=MovieSceneScriptingFloatChannel, const [1,1]}
         ]
        }]
      }
     ]
    }
    Image_ShadowFor3dMap {
     id=#5
     possessed_class=Image
     tracks [
      {
       type=MovieSceneWidgetMaterialTrack
       sections [{
         type=MovieSceneParameterSection
         channels [
          Offset and AlphaR {type=MovieSceneScriptingFloatChannel, const [0.85,2]}
          Offset and AlphaG {type=MovieSceneScriptingFloatChannel, const [1.6,2]}
          Offset and AlphaB {type=MovieSceneScriptingFloatChannel, keys [[1,0.3],[5,0.4]]}
          Offset and AlphaA {type=MovieSceneScriptingFloatChannel, const [0.35,2]}
         ]
        }]
      }
      {
       type=MovieScene2DTransformTrack
       sections [{
         type=MovieScene2DTransformSection
         channels [
          Translation.X {type=MovieSceneScriptingFloatChannel}
          Translation.Y {type=MovieSceneScriptingFloatChannel}
          Angle {type=MovieSceneScriptingFloatChannel}
          Scale.X {type=MovieSceneScriptingFloatChannel}
          Scale.Y {type=MovieSceneScriptingFloatChannel}
          Shear.X {type=MovieSceneScriptingFloatChannel}
          Shear.Y {type=MovieSceneScriptingFloatChannel}
         ]
        }]
      }
     ]
    }
   ]}
 }
 OnStartSelected {
  binding_count=7
  length_seconds=0.3
  object=&self:OnStartSelected
  sequence {bindings [
    Text_AllChallenges {
     id=#15
     possessed_class=CommonTextBlock
     tracks [{
       type=MovieSceneFloatTrack
       display_name=Render Opacity
       sections [{type=MovieSceneFloatSection, channels [None {type=MovieSceneScriptingFloatChannel, keys [[3,0],[5,1]]}]}]
      }]
    }
    Image_Flare {
     id=#1
     possessed_class=Image
     tracks [
      {
       type=MovieSceneFloatTrack
       display_name=Render Opacity
       sections [{type=MovieSceneFloatSection, channels [None {type=MovieSceneScriptingFloatChannel, keys [[0,0],[1,1],[2,1],[3,0.85],[6,0]]}]}]
      }
      {
       type=MovieScene2DTransformTrack
       sections [{
         type=MovieScene2DTransformSection
         channels [
          Translation.X {type=MovieSceneScriptingFloatChannel}
          Translation.Y {type=MovieSceneScriptingFloatChannel}
          Angle {type=MovieSceneScriptingFloatChannel}
          Scale.X {type=MovieSceneScriptingFloatChannel, keys [[1.9977,0.75],[2,0.825],[5,1.5],[6,1.75]]}
          Scale.Y {type=MovieSceneScriptingFloatChannel, keys [[1.9977,0.75],[2,0.825],[5,1.5],[6,1.75]]}
          Shear.X {type=MovieSceneScriptingFloatChannel}
          Shear.Y {type=MovieSceneScriptingFloatChannel}
         ]
        }]
      }
      {
       type=MovieSceneWidgetMaterialTrack
       sections [{type=MovieSceneParameterSection, channels [Content-Offset {type=MovieSceneScriptingFloatChannel, keys [[2,1],[3,0.2],[6,0.2]]}]}]
      }
     ]
    }
    Overlay_Internal {
     id=#2
     possessed_class=Overlay
     tracks [
      {
       type=MovieScene2DTransformTrack
       sections [{
         type=MovieScene2DTransformSection
         channels [
          Translation.X {type=MovieSceneScriptingFloatChannel}
          Translation.Y {type=MovieSceneScriptingFloatChannel}
          Angle {type=MovieSceneScriptingFloatChannel}
          Scale.X {type=MovieSceneScriptingFloatChannel, keys [[2,0.5],[5,1.28],[6,1.25]]}
          Scale.Y {type=MovieSceneScriptingFloatChannel, keys [[2,0.5],[5,1.28],[6,1.25]]}
          Shear.X {type=MovieSceneScriptingFloatChannel}
          Shear.Y {type=MovieSceneScriptingFloatChannel}
         ]
        }]
      }
      {
       type=MovieSceneFloatTrack
       display_name=Render Opacity
       sections [{type=MovieSceneFloatSection, channels [None {type=MovieSceneScriptingFloatChannel, keys [[2,0],[5,1]]}]}]
      }
     ]
    }
    CompletedCountText {
     id=#16
     possessed_class=CommonTextBlock
     tracks [{
       type=MovieSceneFloatTrack
       display_name=Render Opacity
       sections [{type=MovieSceneFloatSection, channels [None {type=MovieSceneScriptingFloatChannel, keys [[4,0],[6,1]]}]}]
      }]
    }
    Image_Icon {
     id=#3
     possessed_class=Image
     tracks [{
       type=MovieSceneWidgetMaterialTrack
       sections [{
         type=MovieSceneParameterSection
         channels [
          Selected {type=MovieSceneScriptingFloatChannel, keys [[2,0],[5,1]]}
          Bot-Radius-S1 {type=MovieSceneScriptingFloatChannel, keys [[2,1.025],[5,1.04]]}
          Bot-Radius-S2 {type=MovieSceneScriptingFloatChannel, keys [[2,0.985],[4,0.9775],[5,1.005]]}
          Inner-Value-S1 {type=MovieSceneScriptingFloatChannel, const [0.25,2]}
          Inner-Value-S2 {type=MovieSceneScriptingFloatChannel, const [0,3]}
          Left-Inner-OffSet-S1 {type=MovieSceneScriptingFloatChannel, keys [[2,0.085],[5,0.065]]}
          Left-Inner-OffSet-S2 {type=MovieSceneScriptingFloatChannel, keys [[2,0.125],[4,0.155],[5,0.11]]}
          Right-Inner-OffSet-S1 {type=MovieSceneScriptingFloatChannel, keys [[2,0.115],[5,0.09]]}
          Right-Inner-OffSet-S2 {type=MovieSceneScriptingFloatChannel, keys [[2,0.155],[4,0.185],[5,0.14]]}
          Top-Inner-OffSet-S1 {type=MovieSceneScriptingFloatChannel, keys [[2,0.085],[5,0.065]]}
          Top-Inner-OffSet-S2 {type=MovieSceneScriptingFloatChannel, keys [[2,0.125],[4,0.155],[5,0.11]]}
          Icon-State {type=MovieSceneScriptingFloatChannel, keys [[2,0],[5,1]]}
         ]
        }]
      }]
    }
    CommonText_Stack_Count {
     id=#4
     possessed_class=CommonTextBlock
     tracks [
      {
       type=MovieScene2DTransformTrack
       sections [{
         type=MovieScene2DTransformSection
         channels [
          Translation.X {type=MovieSceneScriptingFloatChannel}
          Translation.Y {type=MovieSceneScriptingFloatChannel}
          Angle {type=MovieSceneScriptingFloatChannel}
          Scale.X {type=MovieSceneScriptingFloatChannel, keys [[2,1],[6,0.95]]}
          Scale.Y {type=MovieSceneScriptingFloatChannel, keys [[2,1],[6,0.95]]}
          Shear.X {type=MovieSceneScriptingFloatChannel}
          Shear.Y {type=MovieSceneScriptingFloatChannel}
         ]
        }]
      }
      {
       type=MovieSceneColorTrack
       display_name=Color and Opacity
       sections [{
         type=MovieSceneColorSection
         channels [
          Color.R {type=MovieSceneScriptingFloatChannel, keys [[2,1],[6,0.0024]]}
          Color.G {type=MovieSceneScriptingFloatChannel, keys [[2,1],[6,0.0395]]}
          Color.B {type=MovieSceneScriptingFloatChannel, keys [[2,1],[6,0.1329]]}
          Color.A {type=MovieSceneScriptingFloatChannel, const [1,2]}
         ]
        }]
      }
     ]
    }
    Image_ShadowFor3dMap {
     id=#5
     possessed_class=Image
     tracks [
      {
       type=MovieSceneWidgetMaterialTrack
       sections [{
         type=MovieSceneParameterSection
         channels [
          Offset and AlphaR {type=MovieSceneScriptingFloatChannel, const [0.85,2]}
          Offset and AlphaG {type=MovieSceneScriptingFloatChannel, const [1.6,2]}
          Offset and AlphaB {type=MovieSceneScriptingFloatChannel, keys [[1,0.3],[5,0.4]]}
          Offset and AlphaA {type=MovieSceneScriptingFloatChannel, const [0.35,2]}
         ]
        }]
      }
      {
       type=MovieScene2DTransformTrack
       sections [{
         type=MovieScene2DTransformSection
         channels [
          Translation.X {type=MovieSceneScriptingFloatChannel}
          Translation.Y {type=MovieSceneScriptingFloatChannel}
          Angle {type=MovieSceneScriptingFloatChannel}
          Scale.X {type=MovieSceneScriptingFloatChannel}
          Scale.Y {type=MovieSceneScriptingFloatChannel}
          Shear.X {type=MovieSceneScriptingFloatChannel}
          Shear.Y {type=MovieSceneScriptingFloatChannel}
         ]
        }]
      }
     ]
    }
   ]}
 }
]
graphs [
 EventGraph {nodes [
   n1 = StopAnimation(InAnimation=Get OnStartAnim() -> (OnStartAnim:Widget Animation Object Reference))
   n2 = Event OnSetIsInGame!() -> (OutputDelegate, then, bIsInGame:bool) {is=title K2Node_Event}
   SetVisibility(
    execute=[@n21.then, @n1.then]
    self=Get Image_ShadowFor3dMap() -> (Image_ShadowFor3dMap:Image Object Reference)
    InVisibility=Select(Option 0=SelfHitTestInvisible, Option 1=Collapsed, Index=@n2.bIsInGame)
   )
   Event OnBeginIntroAnimation() {is=title K2Node_Event}
   > SetVisibility(
    self=[Get Overlay_1() -> (Overlay_1:Overlay Object Reference), Get VerticalBox_2() -> (VerticalBox_2:Vertical Box Object Reference)]
    InVisibility=Hidden
   )
   > SetVisibility(self=[Get Overlay_1(), Get VerticalBox_2()], InVisibility=SelfHitTestInvisible)
   > PlayAnimation(InAnimation=Get OnStartAnim())
   n3 = Event OnChangeBangVisibility!() -> (OutputDelegate, then, bShouldShow:bool) {is=title K2Node_Event}
   > SetVisibility(
    self=Get Image_Bang() -> (Image_Bang:Fort Mobile Image Object Reference)
    InVisibility=Select(Option 0=Collapsed, Option 1=SelfHitTestInvisible, Index=@n3.bShouldShow)
   )
   n4 = Event OnChallengeSet!() -> (
    OutputDelegate
    then
    bIsComplete:bool
    bIsLocked:bool
    bHasIconOverride:bool
    Progress:float
    bShouldShowBang:bool
    bIsPinned:bool
    bChallengeSelected:bool
    bIsTracked:bool
    PinnedQuestIndex:int=0
   ) {is=title K2Node_Event}
   > SetVisibility(self=Get Image_Bang(), InVisibility=Select(Option 0=Collapsed, Option 1=SelfHitTestInvisible, Index=@n4.bShouldShowBang))
   > SetMaterialState!(execute, self, Completed:bool=@n4.bIsComplete, Lock:bool=@n4.bIsLocked, HasIconOverride:bool=@n4.bHasIconOverride) -> (then)
   > n5 = GetDynamicMaterial(self=Get Image_Icon() -> (Image_Icon:Image Object Reference))
   > SetScalarParameterValue(self=@n5, ParameterName=Progress, Value=@n4.Progress)
   n6 = GetSelectStackedDataIndex!(self:Athena Map Navigable Icon Stack Of Challenges Object Reference) -> (ReturnValue:int=0)
   > "SetText (Text)"(
    self=Get CommonText_Stack() -> (CommonText_Stack:Common Text Object Reference)
    InText="To Text (Integer)"(Value=int + int(A=@n6, B=1) -> (ReturnValue=0) {is=title K2Node_CommutativeAssociativeBinaryOperator})
   )
   n7 = OR Boolean!(A=@n4.bIsComplete, B=@n4.bIsLocked, C:bool=@n4.bHasIconOverride) -> (ReturnValue)
   > Set bEnforceHideCount(bEnforceHideCount:bool=@n7) -> (Output_Get:bool)
   > SetVisibility(
    self=Get CommonText_Stack_Count() -> (CommonText_Stack_Count:Common Text Object Reference)
    InVisibility=Select(Option 0=SelfHitTestInvisible, Option 1=Hidden, Index=@n7)
   )
   n8 = Event OnChallengeIconOverriden!() -> (OutputDelegate, then, Icon:Texture 2D Object Reference) {is=title K2Node_Event}
   > n9 = GetDynamicMaterial(self=Get Image_Icon())
   > n10 = Get Texture Parameter Value!(execute, self:Material Instance Dynamic Object Reference=@n9, ParameterName:Name=Icon) -> (then, ReturnValue:Texture Object Reference) {is=title K2Node_CallFunction}
   > Set DefaultIcon(DefaultIcon:Texture Object Reference=@n10) -> (Output_Get:Texture Object Reference)
   > SetTextureParameterValue(self=@n9, ParameterName=Icon, Value=@n8.Icon)
   n11 = Event OnSelected!() -> (OutputDelegate, then, bUserInteraction:bool) {is=title K2Node_Event}
   > SetVisibility(self=Get Image_Flare() -> (Image_Flare:Fort Mobile Image Object Reference), InVisibility=SelfHitTestInvisible)
   n12 = Select(
    Option 0:Widget Animation Object Reference=Get OnSelectedAnim() -> (OnSelectedAnim:Widget Animation Object Reference)
    Option 1:Widget Animation Object Reference=Get OnSelectAnim-FromHover() -> (OnSelectAnim-FromHover:Widget Animation Object Reference)
    Index=IsIconSelected!(self:Athena Map Navigable Icon Object Reference) -> (ReturnValue:bool)
   ) -> (ReturnValue:Widget Animation Object Reference)
   > PlayAnimation(
    InAnimation=@n12
    StartAtTime=Select(Option 0:double=GetEndTime(self=@n12), Option 1:double=0, Index=@n11.bUserInteraction) -> (ReturnValue:double)
   )
   > PlayAnimation(InAnimation=Get Flare-Loop() -> (Flare-Loop:Widget Animation Object Reference), NumLoopsToPlay=0)
   > PlaySound2D(Sound=/Game/Sounds/Fort_UI_Sounds/S12_Frontend/UI_Table_HoverButton_01_Cue)
   Event OnUnselected!() -> (OutputDelegate, then, bUserInteraction:bool) {is=title K2Node_Event}
   > SetVisibility(self=Get Image_Flare(), InVisibility=Collapsed)
   > PlayAnimation(InAnimation=Get OnSelectedAnim(), PlayMode=Reverse)
   n13 = Event OnSetChallengesInfo!() -> (OutputDelegate, then, ChallengesNum:int=0, CountOfCompletedChallenges:int=0) {is=title K2Node_Event}
   > n14 = Branch(Condition="integer > integer"(A=@n13.ChallengesNum, B=0) {is=title K2Node_CallFunction})
   > "SetText (Text)"(
    self=Get CompletedCountText() -> (CompletedCountText:Common Text Object Reference)
    InText=Format Text!(Format="LOCTEXT(\"Completed {0} of {1}\")", 0:int=@n13.CountOfCompletedChallenges, 1:int=@n13.ChallengesNum) -> (Result)
   )
   > SetVisibility(self=Get CompletedCountText(), InVisibility=SelfHitTestInvisible)
   SetVisibility(execute=@n14.else, self=Get CompletedCountText(), InVisibility=Collapsed)
   n15 = Event OnChangedStackData!() -> (OutputDelegate, then, NewStackCount:int=0, NonCompletedCount:int=0) {is=title K2Node_Event}
   > n16 = Branch(Condition="integer > integer"(A=@n15.NonCompletedCount, B=0))
   > "SetText (Text)"(self=Get CommonText_Stack_Count(), InText="To Text (Integer)"(Value=@n15.NonCompletedCount))
   > SetVisibility(
    self=Get CommonText_Stack_Count()
    InVisibility=Select(
     Option 0:ESlateVisibility Enum=SelfHitTestInvisible
     Option 1:ESlateVisibility Enum=Hidden
     Index=Get bEnforceHideCount() -> (bEnforceHideCount:bool)
    ) -> (ReturnValue:ESlateVisibility Enum)
   )
   SetVisibility(execute=@n16.else, self=Get CommonText_Stack_Count(), InVisibility=Hidden)
   Event OnChallengeReset() {is=title K2Node_Event}
   > n17 = GetDynamicMaterial(self=Get Image_Icon())
   > SetTextureParameterValue(self=@n17, ParameterName=Icon, Value=Get DefaultIcon() -> (DefaultIcon:Texture Object Reference))
   > SetScalarParameterValue(self=@n17, ParameterName=bIsLocked, Value=0)
   > SetScalarParameterValue(self=@n17, ParameterName=Progress, Value=0)
   Event Construct()
   > SetVisibility(self=Get CompletedCountText(), InVisibility=Hidden)
   Event OnHovered!() -> (OutputDelegate, then, bUserInteraction:bool) {is=title K2Node_Event}
   > PlaySound2D(Sound=/Game/Sounds/Fort_UI_Sounds/S12_Frontend/UI_Challenge_Table_Hover_01_Cue)
   > n18 = Branch(Condition=IsIconSelected())
   PlayAnimationForward(execute=@n18.else, InAnimation=Get OnHover() -> (OnHover:Widget Animation Object Reference))
   Event OnUnhovered!() -> (OutputDelegate, then, bUserInteraction:bool) {is=title K2Node_Event}
   > n19 = Branch(Condition=IsIconSelected())
   PlayAnimationReverse(execute=@n19.else, InAnimation=Get OnHover())
   SetVisibility(
    self=Get Image_ShadowFor3dMap()
    InVisibility=Select(
     Option 0=Collapsed
     Option 1=SelfHitTestInvisible
     Index=Event OnSetHasLocation!() -> (OutputDelegate, then, bHasLocation:bool) {is=title K2Node_Event}.bHasLocation
    )
   )
   n20 = GetDynamicMaterial(execute=@n2.then, self=Get Image_Icon())
   > SetScalarParameterValue(self=@n20, ParameterName=bIsIngame, Value=Select(Option 0:double=0, Option 1:double=1, Index=@n2.bIsInGame) -> (ReturnValue:double))
   > SetVisibility(self=Get Overlay_1(), InVisibility=SelfHitTestInvisible)
   > n21 = StopAllAnimations(self:User Widget Object Reference)
  ]}
 SetAnywhereState {nodes [
   SetAnywhereState() {is=title K2Node_FunctionEntry}
   > SetDesiredSizeOverride!(
    execute
    self:Image Object Reference=[Get Image_ShadowFor3dMap(), Get Image_Icon(), Get Image_Flare()]
    DesiredSize:Vector 2D Structure=(X=96,Y=96)
   ) -> (then)
   n1 = Get CommonText_Stack_Count()
   n2 = Break Slate Font Info!(SlateFontInfo:"Slate Font Info Structure (by ref)"=Get Font!(self:Text Object Reference=@n1) -> (Font:Slate Font Info Structure)) -> (
    FontObject:Object Reference
    FontMaterial:Object Reference
    OutlineSettings:Font Outline Settings Structure
    TypefaceFontName:Name
    Size:float=~
    LetterSpacing:int
    SkewAmount:float=~
    bForceMonospaced:bool=~
    bMaterialIsStencil:bool=~
    MonospacedWidth:float=~
   ) {is=title K2Node_BreakStruct}
   n3 = MakeLiteralInt!(Value:int=32) -> (ReturnValue:int=0)
   > SetFont(
    self=@n1
    InFontInfo=MakeSlateFontInfo!(
     FontObject:Object Reference=@n2
     FontMaterial:Object Reference=@n2.FontMaterial
     OutlineSettings:Font Outline Settings Structure=@n2.OutlineSettings
     TypefaceFontName:Name=@n2.TypefaceFontName
     Size:float="To Float (Integer)"(InInt=@n3)
     LetterSpacing:int=0
     SkewAmount:float=0
     bForceMonospaced:bool=false
     bMaterialIsStencil:bool=false
     MonospacedWidth:float=1
    ) -> (ReturnValue:Slate Font Info Structure)
   )
   > SetVisibility(self=Get Image_ShadowFor3dMap(), InVisibility=Collapsed)
  ]}
 SetMaterialState {nodes [
   n1 = SetMaterialState!() -> (then, Completed:bool=~, Lock:bool=~, HasIconOverride:bool=~) {is=title K2Node_FunctionEntry}
   > n2 = Branch(Condition=@n1)
   > SetBrushFromMaterial(self=Get Image_Icon(), Material=/Game/Athena/UI/Challenges/Art/MapIcons/M_UI_MapMarker_Anywhere_Comp)
   n3 = Branch(execute=@n2.else, Condition=@n1.Lock)
   > SetBrushFromMaterial(self=Get Image_Icon(), Material=/Game/Athena/UI/Challenges/Art/MapIcons/M_UI_MapMarker_Anywhere_Lock)
   n4 = Branch(execute=@n3.else, Condition=@n1.HasIconOverride)
   > SetBrushFromMaterial(self=Get Image_Icon(), Material=/Game/Athena/UI/Challenges/Art/MapIcons/M_UI_MapMarker_Anywhere_Icon)
   SetBrushFromMaterial(execute=@n4.else, self=Get Image_Icon(), Material=/Game/Athena/UI/Challenges/Art/MapIcons/M_UI_MapMarker_Anywhere)
  ]}
]
variables [
 bEnforceHideCount {type=boolean}
 DefaultIcon {type=Texture, default=null}
]
widgets {
 info {
  inheritedWidgetCount=3
  parentClass=/Script/FortniteUI.AthenaMapNavigableIconStackOfNoLocationChallenges
  rootWidgetClass=&HorizontalBox
  widgetCount=16
 }
 widgets [
  HorizontalBox_1 {widgetClassPath=&HorizontalBox, widget=HorizontalBox_1}
  Overlay_1 {
   widgetClassPath=&Overlay
   bIsVariable=true
   parent=HorizontalBox_1
   widget=Overlay_1
   slot {class=HorizontalBoxSlot, properties {horizontalAlignment=HAlign_Left, verticalAlignment=VAlign_Center}}
  }
  Image_ShadowFor3dMap {
   widgetClassPath=&Image
   bIsVariable=true
   parent=Overlay_1
   widget=Image_ShadowFor3dMap
   properties {
    visibility=SelfHitTestInvisible
    brush {
     bIsDynamicallyLoaded=false
     drawAs=Image
     imageType=NoImage
     mirroring=NoMirror
     resourceName=None
     resourceObject=/Game/Athena/UI/Challenges/Art/MapIcons/M_UI_MapMarker_Shadow
     tiling=NoTile
     imageSize=<96 96>
     margin=<ltrb 0 0 0 0>
     outlineSettings {
      bUseBrushTransparency=false
      roundingType=HalfHeightRadius
      width=0
      color {colorUseRule=UseColor_Specified, specifiedColor=<rgba 0 0 0 0>}
      cornerRadii=<0 0 0 0>
     }
     tintColor {colorUseRule=UseColor_Specified, specifiedColor=<rgba 1 1 1 1>}
     uVRegion {bIsValid=false, max=<0 0>, min=<0 0>}
    }
    renderTransform {angle=0, scale=<2 2>, shear=<0 0>, translation=<0 -16>}
   }
   slot {class=OverlaySlot, properties {horizontalAlignment=HAlign_Center, verticalAlignment=VAlign_Center}}
  }
  Overlay_Internal {
   widgetClassPath=&Overlay
   parent=Overlay_1
   widget=Overlay_Internal
   slot {class=OverlaySlot, properties {horizontalAlignment=HAlign_Center}}
  }
  Image_Flare {
   widgetClassPath=/Script/FortniteUI.FortMobileImage
   bIsVariable=true
   parent=Overlay_Internal
   widget=Image_Flare
   properties {
    mobileSizeMultiplier=1.5
    renderOpacity=0
    visibility=Collapsed
    brush {
     bIsDynamicallyLoaded=false
     drawAs=Image
     imageType=NoImage
     mirroring=NoMirror
     resourceName=None
     resourceObject=/Game/Athena/UI/Challenges/Art/MapIcons/M_UI_MapMarker_Glow
     tiling=NoTile
     imageSize=<96 96>
     margin=<ltrb 0 0 0 0>
     outlineSettings {
      bUseBrushTransparency=false
      roundingType=HalfHeightRadius
      width=0
      color {colorUseRule=UseColor_Specified, specifiedColor=<rgba 0 0 0 0>}
      cornerRadii=<0 0 0 0>
     }
     tintColor {colorUseRule=UseColor_Specified, specifiedColor=<rgba 1 1 1 1>}
     uVRegion {bIsValid=false, max=<0 0>, min=<0 0>}
    }
    renderTransform {angle=0, scale=<0.825 0.825>, shear=<0 0>, translation=<-1 -9>}
   }
   slot {class=OverlaySlot, properties {horizontalAlignment=HAlign_Center, verticalAlignment=VAlign_Center}}
  }
  Image_Icon {
   widgetClassPath=&FortMobileImage
   bInherited=true
   bIsVariable=true
   parent=Overlay_Internal
   widget=Image_Icon
   properties {
    mobileSizeMultiplier=1.5
    visibility=SelfHitTestInvisible
    brush {
     bIsDynamicallyLoaded=false
     drawAs=Image
     imageType=NoImage
     mirroring=NoMirror
     resourceName=None
     resourceObject=&M_UI_MapMarker_Anywhere
     tiling=NoTile
     imageSize=<96 96>
     margin=<ltrb 0 0 0 0>
     outlineSettings {
      bUseBrushTransparency=false
      roundingType=HalfHeightRadius
      width=0
      color {colorUseRule=UseColor_Specified, specifiedColor=<rgba 0 0 0 0>}
      cornerRadii=<0 0 0 0>
     }
     tintColor {colorUseRule=UseColor_Specified, specifiedColor=<rgba 1 1 1 1>}
     uVRegion {bIsValid=false, max=<0 0>, min=<0 0>}
    }
    renderTransformPivot=<0.5 1>
   }
   slot {class=OverlaySlot}
  }
  CommonActionWidget_Cycle {
   widgetClassPath=&CommonActionWidget
   bInherited=true
   bIsVariable=true
   parent=Overlay_Internal
   widget=CommonActionWidget_Cycle
   properties {visibility=Collapsed, inputActions [{dataTable=&InputActionDataTable, rowName=GenericForward}]}
   slot {class=OverlaySlot, properties {horizontalAlignment=HAlign_Right}}
  }
  Border_0 {
   widgetClassPath=&Border
   bIsVariable=true
   parent=Overlay_Internal
   widget=Border_0
   properties {
    visibility=Collapsed
    background {
     bIsDynamicallyLoaded=false
     drawAs=Image
     imageType=NoImage
     mirroring=NoMirror
     resourceName=None
     resourceObject=/Game/Athena/UI/Challenges/Art/MapIcons/M_UI_MarkerState-Counter
     tiling=NoTile
     imageSize=<32 32>
     margin=<ltrb 0 0 0 0>
     outlineSettings {
      bUseBrushTransparency=false
      roundingType=HalfHeightRadius
      width=0
      color {colorUseRule=UseColor_Specified, specifiedColor=<rgba 0 0 0 0>}
      cornerRadii=<0 0 0 0>
     }
     tintColor {colorUseRule=UseColor_Specified, specifiedColor=<rgba 1 1 1 1>}
     uVRegion {bIsValid=false, max=<0 0>, min=<0 0>}
    }
    padding=<ltrb 4 4 4 4>
    renderTransform {angle=0, scale=<1 1>, shear=<0 0>, translation=<0 -5>}
   }
   slot {class=OverlaySlot, properties {horizontalAlignment=HAlign_Right, verticalAlignment=VAlign_Bottom}}
  }
  HorizontalBox_0 {
   widgetClassPath=&HorizontalBox
   parent=Border_0
   widget=HorizontalBox_0
   slot {class=BorderSlot, properties {padding=<ltrb 4 4 4 4>}}
  }
  CommonText_Stack {
   widgetClassPath=&CommonTextBlock
   bIsVariable=true
   parent=HorizontalBox_0
   widget=CommonText_Stack
   properties {
    text=88
    visibility=Collapsed
    font {
     bForceMonospaced=false
     bMaterialIsStencil=false
     fontMaterial=null
     fontObject=&BurbankSmall
     letterSpacing=0
     monospacedWidth=1
     size=15
     skewAmount=0
     typefaceFontName=Bold
     outlineSettings {
      bApplyOutlineToDropShadows=false
      bMiteredCorners=false
      bSeparateFillAlpha=false
      outlineMaterial=null
      outlineSize=0
      outlineColor=<rgba 0 0 0 1>
     }
    }
    shadowOffset=<0 0>
   }
   slot {class=HorizontalBoxSlot}
  }
  TextOf {
   widgetClassPath=&CommonTextBlock
   parent=HorizontalBox_0
   widget=TextOf
   properties {
    text=" of "
    visibility=Collapsed
    font {
     bForceMonospaced=false
     bMaterialIsStencil=false
     fontMaterial=null
     fontObject=&BurbankSmall
     letterSpacing=0
     monospacedWidth=1
     size=15
     skewAmount=0
     typefaceFontName=Bold
     outlineSettings {
      bApplyOutlineToDropShadows=false
      bMiteredCorners=false
      bSeparateFillAlpha=false
      outlineMaterial=null
      outlineSize=0
      outlineColor=<rgba 0 0 0 1>
     }
    }
    shadowOffset=<0 0>
   }
   slot {class=HorizontalBoxSlot}
  }
  Image_Bang {
   widgetClassPath=&FortMobileImage
   bIsVariable=true
   parent=Overlay_Internal
   widget=Image_Bang
   properties {
    mobileSizeMultiplier=1.5
    visibility=Collapsed
    brush {
     bIsDynamicallyLoaded=false
     drawAs=Image
     imageType=NoImage
     mirroring=NoMirror
     resourceName=None
     resourceObject=/Game/Athena/UI/Challenges/Art/MapIcons/M_UI_MarkerState-Bang
     tiling=NoTile
     imageSize=<32 32>
     margin=<ltrb 0 0 0 0>
     outlineSettings {
      bUseBrushTransparency=false
      roundingType=HalfHeightRadius
      width=0
      color {colorUseRule=UseColor_Specified, specifiedColor=<rgba 0 0 0 0>}
      cornerRadii=<0 0 0 0>
     }
     tintColor {colorUseRule=UseColor_Specified, specifiedColor=<rgba 1 1 1 1>}
     uVRegion {bIsValid=false, max=<0 0>, min=<0 0>}
    }
    renderTransform {angle=0, scale=<1 1>, shear=<0 0>, translation=<4 -4>}
    renderTransformPivot=<0 1>
   }
   slot {class=OverlaySlot, properties {horizontalAlignment=HAlign_Right}}
  }
  CommonText_Stack_Count {
   widgetClassPath=&CommonTextBlock
   bIsVariable=true
   parent=Overlay_Internal
   widget=CommonText_Stack_Count
   properties {
    justification=Center
    mobileFontSizeMultiplier=1.5
    font {
     bForceMonospaced=false
     bMaterialIsStencil=false
     fontMaterial=null
     fontObject=&BurbankSmall
     letterSpacing=0
     monospacedWidth=1
     size=24
     skewAmount=0
     typefaceFontName=Bold
     outlineSettings {
      bApplyOutlineToDropShadows=false
      bMiteredCorners=false
      bSeparateFillAlpha=false
      outlineMaterial=null
      outlineSize=0
      outlineColor=<rgba 0 0 0 1>
     }
    }
    renderTransform {angle=0, scale=<1 1>, shear=<-12 0>, translation=<0 0>}
    shadowOffset=<0 0>
   }
   slot {class=OverlaySlot, properties {horizontalAlignment=HAlign_Center, padding=<ltrb 0 16 0 0>}}
  }
  VerticalBox_2 {
   widgetClassPath=&VerticalBox
   bIsVariable=true
   parent=HorizontalBox_1
   widget=VerticalBox_2
   slot {class=HorizontalBoxSlot, properties {horizontalAlignment=HAlign_Left, verticalAlignment=VAlign_Top, padding=<ltrb 14 6 0 0>}}
  }
  Text_AllChallenges {
   widgetClassPath=&CommonTextBlock
   bInherited=true
   bIsVariable=true
   parent=VerticalBox_2
   widget=Text_AllChallenges
   properties {
    mobileFontSizeMultiplier=1.34
    text=ALL CHALLENGES
    textTransformPolicy=ToUpper
    wrapTextAt=320
    font {
     bForceMonospaced=false
     bMaterialIsStencil=false
     fontMaterial=null
     fontObject=&BurbankSmall
     letterSpacing=0
     monospacedWidth=1
     size=18
     skewAmount=0
     typefaceFontName=Bold
     outlineSettings {
      bApplyOutlineToDropShadows=false
      bMiteredCorners=false
      bSeparateFillAlpha=false
      outlineMaterial=null
      outlineSize=0
      outlineColor=<rgba 0 0 0 1>
     }
    }
    renderTransform {angle=0, scale=<1 1>, shear=<-8 0>, translation=<0 0>}
    shadowOffset=<0 0>
   }
   slot {class=VerticalBoxSlot, properties {horizontalAlignment=HAlign_Left, verticalAlignment=VAlign_Top}}
  }
  CompletedCountText {
   widgetClassPath=&CommonTextBlock
   bIsVariable=true
   parent=VerticalBox_2
   widget=CompletedCountText
   properties {
    justification=Center
    text=Completed 0 of 99
    textTransformPolicy=ToUpper
    colorAndOpacity {colorUseRule=UseColor_Specified, specifiedColor=<rgba 0.5 0.9582 1 1>}
    font {
     bForceMonospaced=false
     bMaterialIsStencil=false
     fontMaterial=null
     fontObject=&BurbankSmall
     letterSpacing=0
     monospacedWidth=1
     size=14
     skewAmount=0
     typefaceFontName=Bold
     outlineSettings {
      bApplyOutlineToDropShadows=false
      bMiteredCorners=false
      bSeparateFillAlpha=false
      outlineMaterial=null
      outlineSize=0
      outlineColor=<rgba 1 1 1 1>
     }
    }
    renderTransform {angle=0, scale=<1 1>, shear=<-8 0>, translation=<0 0>}
    shadowOffset=<0 0>
   }
   slot {class=VerticalBoxSlot, properties {horizontalAlignment=HAlign_Left, verticalAlignment=VAlign_Top}}
  }
 ]
}
