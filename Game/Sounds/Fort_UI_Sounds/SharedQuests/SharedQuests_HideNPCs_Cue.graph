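# SharedQuests_HideNPCs_Cue is a UI sound cue for hiding NPCs in the shared quests interface.
# A single wave player feeds the root with the ToggleNPC hide wave, routed through the Fort_UI sound class at reduced volume.
# A front-end UI menu concurrency set is assigned, alongside stop-oldest override settings for a single instance.
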SoundCue /Game/Sounds/Fort_UI_Sounds/SharedQuests/SharedQuests_HideNPCs_Cue
root=WavePlayer_0
nodes [{id=WavePlayer_0, properties {soundWaveAssetPtr=&/SharedQuests_ToggleNPC_Hide_01}}]
properties {
 maxDistance=2097152
 soundClassObject=&Fort_UI
 volumeMultiplier=0.7
 concurrencyOverrides {maxCount=1, resolutionRule=StopOldest}
 concurrencySet [/Game/Sounds/Concurrency_Rules/FrontEnd_UI_Menu_Max_03_Stop_Oldest]
}
