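# Widget Blueprint for the full-screen Athena map's browsing controls legend, showing prompts to pan, zoom, and place or remove markers and danger pings.
# Each row pairs labels with action or keybind widgets, using platform and input visibility wrappers to switch between mouse, gamepad and touch hints.
# The event graph is empty; the native parent class supplies the inherited bound widgets.
# Sound defaults cover quest tracking and NPC show/hide, unused by any graph here.
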WidgetBlueprint /FortniteCoreUI/HUD/MiniMap/BP_AthenaFullScreenMapBrowsingControlsWidget
graphs [EventGraph {links [], nodes []}]
variables [
 HideNPCsSound {type=SoundBase, default=/Game/Sounds/Fort_UI_Sounds/SharedQuests/SharedQuests_HideNPCs_Cue}
 ShowNPCsSound {type=SoundBase, default=/Game/Sounds/Fort_UI_Sounds/SharedQuests/SharedQuests_ShowNPCs_Cue}
 TrackSound {type=SoundBase, default=/Game/Sounds/Fort_UI_Sounds/SharedQuests/SharedQuests_PinQuest_Cue}
 UntrackSound {type=SoundBase, default=/Game/Sounds/Fort_UI_Sounds/SharedQuests/SharedQuests_UnpinQuest_Cue}
]
widgets {
 info {
  inheritedWidgetCount=10
  parentClass=/Script/FortniteUI.AthenaFullScreenMapBrowsingControlsWidget
  rootWidgetClass=&Overlay
  widgetCount=47
 }
 widgets [
  Overlay_1 {widgetClassPath=&Overlay, widget=Overlay_1}
  GridPanel_buttons {
   widgetClassPath=&VerticalBox
   parent=Overlay_1
   widget=GridPanel_buttons
   slot {class=OverlaySlot, properties {horizontalAlignment=HAlign_Right, verticalAlignment=VAlign_Center}}
  }
  Panel_PanGroup {
   widgetClassPath=&SizeBox
   bInherited=true
   bIsVariable=true
   parent=GridPanel_buttons
   widget=Panel_PanGroup
   properties {bOverride_MaxDesiredWidth=true, maxDesiredWidth=305}
   slot {class=VerticalBoxSlot, properties {padding=<ltrb 0 0 0 8>}}
  }
  VerticalBox_3 {widgetClassPath=&VerticalBox, parent=Panel_PanGroup, widget=VerticalBox_3, slot {class=SizeBoxSlot}}
  HorizontalBox_145 {widgetClassPath=&HorizontalBox, parent=VerticalBox_3, widget=HorizontalBox_145, slot {class=VerticalBoxSlot}}
  Visibility_PanText {
   widgetClassPath=&CommonVisibilityWidgetLegacy
   parent=HorizontalBox_145
   widget=Visibility_PanText
   properties {bShowForTouch=false, padding=<ltrb 0 0 0 0>}
   slot {class=HorizontalBoxSlot, properties {horizontalAlignment=HAlign_Right, verticalAlignment=VAlign_Center, size {sizeRule=Fill, value=1}}}
  }
  Text_Pan_1 {
   widgetClassPath=&CommonTextBlock
   parent=Visibility_PanText
   widget=Text_Pan_1
   properties {
    justification=Center
    scrollStyle=&TextScrollStyle-Base_C
    text=Pan
    textTransformPolicy=ToUpper
    font {
     bForceMonospaced=false
     bMaterialIsStencil=false
     fontMaterial=null
     fontObject=&BurbankSmall
     letterSpacing=0
     monospacedWidth=1
     size=18
     skewAmount=0
     typefaceFontName=Black
     outlineSettings {
      bApplyOutlineToDropShadows=false
      bMiteredCorners=false
      bSeparateFillAlpha=false
      outlineMaterial=null
      outlineSize=0
      outlineColor=<rgba 0 0 0 1>
     }
    }
    shadowOffset=<0 0>
   }
   slot {class=BorderSlot, properties {padding=<ltrb 0 0 0 0>}}
  }
  Visibility_MousePan {
   widgetClassPath=&CommonVisibilityWidgetLegacy
   parent=HorizontalBox_145
   widget=Visibility_MousePan
   properties {
    bShowForAndroid=false
    bShowForGamepad=false
    bShowForIOS=false
    bShowForTouch=false
    verticalAlignment=VAlign_Center
    padding=<ltrb 0 0 0 0>
   }
   slot {class=HorizontalBoxSlot, properties {padding=<ltrb 4 0 0 0>}}
  }
  Action_MousePan {
   widgetClassPath=&CommonActionWidget
   parent=Visibility_MousePan
   widget=Action_MousePan
   properties {visibility=SelfHitTestInvisible, inputActions [{dataTable=/FortUILibrary/Input/AthenaMiniMapActionDataTable, rowName=MousePan}]}
   slot {class=BorderSlot, properties {verticalAlignment=VAlign_Center, padding=<ltrb 0 0 0 0>}}
  }
  Visibility_GamepadPan {
   widgetClassPath=&CommonVisibilityWidgetLegacy
   parent=HorizontalBox_145
   widget=Visibility_GamepadPan
   properties {bShowForMouseAndKeyboard=false, bShowForTouch=false, verticalAlignment=VAlign_Center, padding=<ltrb 0 0 0 0>}
   slot {class=HorizontalBoxSlot, properties {padding=<ltrb 4 0 0 0>}}
  }
  Action_GamepadPan {
   widgetClassPath=&CommonActionWidget
   parent=Visibility_GamepadPan
   widget=Action_GamepadPan
   properties {inputActions [
     {dataTable=&AthenaMiniMapActionDataTable, rowName=GamepadPanLeftRight}
     {dataTable=&AthenaMiniMapActionDataTable, rowName=GamepadPanUpDown}
    ]}
   slot {class=BorderSlot, properties {verticalAlignment=VAlign_Center, padding=<ltrb 0 0 0 0>}}
  }
  CommonHardwareVisibilityBorder_3 {
   widgetClassPath=&CommonHardwareVisibilityBorder
   parent=VerticalBox_3
   widget=CommonHardwareVisibilityBorder_3
   properties {visibilityQuery {
     autoDescription=" ANY( Input.Touch )"
     tokenStreamVersion=0
     userDescription=Touch
     queryTokenStream [0,1,1,1,0]
     tagDictionary [{tagName=Input.Touch}]
    }}
   slot {class=VerticalBoxSlot, properties {horizontalAlignment=HAlign_Right, verticalAlignment=VAlign_Center, size {sizeRule=Fill, value=1}}}
  }
  Text_Pan_Mobile {
   widgetClassPath=&CommonTextBlock
   parent=CommonHardwareVisibilityBorder_3
   widget=Text_Pan_Mobile
   properties {
    justification=Center
    scrollStyle=&TextScrollStyle-Base_C
    text=Drag to pan
    textTransformPolicy=ToUpper
    font {
     bForceMonospaced=false
     bMaterialIsStencil=false
     fontMaterial=null
     fontObject=&BurbankSmall
     letterSpacing=0
     monospacedWidth=1
     size=16
     skewAmount=0
     typefaceFontName=Black
     outlineSettings {
      bApplyOutlineToDropShadows=false
      bMiteredCorners=false
      bSeparateFillAlpha=false
      outlineMaterial=null
      outlineSize=0
      outlineColor=<rgba 0 0 0 1>
     }
    }
    shadowOffset=<0 0>
   }
   slot {class=BorderSlot, properties {padding=<ltrb 0 0 0 0>}}
  }
  Panel_ZoomGroup {
   widgetClassPath=&SizeBox
   bInherited=true
   bIsVariable=true
   parent=GridPanel_buttons
   widget=Panel_ZoomGroup
   properties {bOverride_MaxDesiredWidth=true, maxDesiredWidth=360}
   slot {class=VerticalBoxSlot, properties {padding=<ltrb 0 0 0 8>}}
  }
  VerticalBox_1 {widgetClassPath=&VerticalBox, parent=Panel_ZoomGroup, widget=VerticalBox_1, slot {class=SizeBoxSlot}}
  ZoomGroup_Others {
   widgetClassPath=&CommonVisibilityWidgetLegacy
   parent=VerticalBox_1
   widget=ZoomGroup_Others
   properties {bShowForTouch=false, padding=<ltrb 0 0 0 0>}
   slot {class=VerticalBoxSlot}
  }
  VerticalBox_0 {
   widgetClassPath=&VerticalBox
   parent=ZoomGroup_Others
   widget=VerticalBox_0
   slot {class=BorderSlot, properties {padding=<ltrb 0 0 0 0>}}
  }
  Visibility_MouseZoom {
   widgetClassPath=&CommonVisibilityWidgetLegacy
   parent=VerticalBox_0
   widget=Visibility_MouseZoom
   properties {
    bShowForAndroid=false
    bShowForGamepad=false
    bShowForIOS=false
    bShowForPS4=false
    bShowForTouch=false
    bShowForXBox=false
    padding=<ltrb 0 0 0 0>
   }
   slot {class=VerticalBoxSlot}
  }
  HorizontalBox {
   widgetClassPath=&HorizontalBox
   parent=Visibility_MouseZoom
   widget=HorizontalBox
   slot {class=BorderSlot, properties {padding=<ltrb 0 0 0 0>}}
  }
  Text_Zoom {
   widgetClassPath=&CommonTextBlock
   parent=HorizontalBox
   widget=Text_Zoom
   properties {
    justification=Center
    scrollStyle=&TextScrollStyle-Base_C
    text=Zoom
    textTransformPolicy=ToUpper
    font {
     bForceMonospaced=false
     bMaterialIsStencil=false
     fontMaterial=null
     fontObject=&BurbankSmall
     letterSpacing=0
     monospacedWidth=1
     size=18
     skewAmount=0
     typefaceFontName=Black
     outlineSettings {
      bApplyOutlineToDropShadows=false
      bMiteredCorners=false
      bSeparateFillAlpha=false
      outlineMaterial=null
      outlineSize=0
      outlineColor=<rgba 0 0 0 1>
     }
    }
    shadowOffset=<0 0>
   }
   slot {class=HorizontalBoxSlot, properties {horizontalAlignment=HAlign_Right, verticalAlignment=VAlign_Center, size {sizeRule=Fill, value=1}}}
  }
  Action_MouseZoom {
   widgetClassPath=&CommonActionWidget
   parent=HorizontalBox
   widget=Action_MouseZoom
   properties {visibility=SelfHitTestInvisible, inputActions [{dataTable=&AthenaMiniMapActionDataTable, rowName=Zoom}]}
   slot {class=HorizontalBoxSlot, properties {horizontalAlignment=HAlign_Right, verticalAlignment=VAlign_Center, padding=<ltrb 4 0 0 0>}}
  }
  CommonHardwareVisibilityBorder_4 {
   widgetClassPath=&CommonHardwareVisibilityBorder
   parent=VerticalBox_0
   widget=CommonHardwareVisibilityBorder_4
   properties {visibilityQuery {
     autoDescription=" ANY( Input.Gamepad )"
     tokenStreamVersion=0
     userDescription=Controller
     queryTokenStream [0,1,1,1,0]
     tagDictionary [{tagName=Input.Gamepad}]
    }}
   slot {class=VerticalBoxSlot}
  }
  VerticalBox_2 {
   widgetClassPath=&VerticalBox
   parent=CommonHardwareVisibilityBorder_4
   widget=VerticalBox_2
   slot {class=BorderSlot, properties {padding=<ltrb 0 0 0 0>}}
  }
  HorizontalBox_2 {
   widgetClassPath=&HorizontalBox
   parent=VerticalBox_2
   widget=HorizontalBox_2
   slot {class=VerticalBoxSlot, properties {padding=<ltrb 0 0 0 8>}}
  }
  Text_ZoomOut {
   widgetClassPath=&CommonTextBlock
   parent=HorizontalBox_2
   widget=Text_ZoomOut
   properties {
    justification=Center
    scrollStyle=&TextScrollStyle-Base_C
    text=Zoom Out
    textTransformPolicy=ToUpper
    font {
     bForceMonospaced=false
     bMaterialIsStencil=false
     fontMaterial=null
     fontObject=&BurbankSmall
     letterSpacing=0
     monospacedWidth=1
     size=18
     skewAmount=0
     typefaceFontName=Black
     outlineSettings {
      bApplyOutlineToDropShadows=false
      bMiteredCorners=false
      bSeparateFillAlpha=false
      outlineMaterial=null
      outlineSize=0
      outlineColor=<rgba 0 0 0 1>
     }
    }
    shadowOffset=<0 0>
   }
   slot {class=HorizontalBoxSlot, properties {horizontalAlignment=HAlign_Right, verticalAlignment=VAlign_Center, size {sizeRule=Fill, value=1}}}
  }
  Keybind_GamepadZoomOut {
   widgetClassPath=/FortUILibrary/Input/KeybindWidget
   parent=HorizontalBox_2
   widget=Keybind_GamepadZoomOut
   properties {boundAction=MapZoomOut_Gamepad}
   slot {class=HorizontalBoxSlot, properties {verticalAlignment=VAlign_Center, padding=<ltrb 4 0 0 0>}}
  }
  HorizontalBox_72 {widgetClassPath=&HorizontalBox, parent=VerticalBox_2, widget=HorizontalBox_72, slot {class=VerticalBoxSlot}}
  Text_ZoomIn {
   widgetClassPath=&CommonTextBlock
   parent=HorizontalBox_72
   widget=Text_ZoomIn
   properties {
    justification=Center
    scrollStyle=&TextScrollStyle-Base_C
    text=Zoom In
    textTransformPolicy=ToUpper
    font {
     bForceMonospaced=false
     bMaterialIsStencil=false
     fontMaterial=null
     fontObject=&BurbankSmall
     letterSpacing=0
     monospacedWidth=1
     size=18
     skewAmount=0
     typefaceFontName=Black
     outlineSettings {
      bApplyOutlineToDropShadows=false
      bMiteredCorners=false
      bSeparateFillAlpha=false
      outlineMaterial=null
      outlineSize=0
      outlineColor=<rgba 0 0 0 1>
     }
    }
    shadowOffset=<0 0>
   }
   slot {class=HorizontalBoxSlot, properties {horizontalAlignment=HAlign_Right, verticalAlignment=VAlign_Center, size {sizeRule=Fill, value=1}}}
  }
  Keybind_GamepadZoomIn {
   widgetClassPath=&KeybindWidget
   parent=HorizontalBox_72
   widget=Keybind_GamepadZoomIn
   properties {boundAction=MapZoomIn_Gamepad}
   slot {class=HorizontalBoxSlot, properties {verticalAlignment=VAlign_Center, padding=<ltrb 4 0 0 0>}}
  }
  ZoomGroup_Mobile {
   widgetClassPath=&CommonHardwareVisibilityBorder
   parent=VerticalBox_1
   widget=ZoomGroup_Mobile
   properties {
    horizontalAlignment=HAlign_Right
    visibilityQuery {
     autoDescription=" ANY( Input.Touch )"
     tokenStreamVersion=0
     userDescription=Touch
     queryTokenStream [0,1,1,1,0]
     tagDictionary [{tagName=Input.Touch}]
    }
   }
   slot {class=VerticalBoxSlot}
  }
  Text_Zoom_Mobile {
   widgetClassPath=&CommonTextBlock
   parent=ZoomGroup_Mobile
   widget=Text_Zoom_Mobile
   properties {
    justification=Center
    scrollStyle=&TextScrollStyle-Base_C
    text=Pinch to zoom
    textTransformPolicy=ToUpper
    font {
     bForceMonospaced=false
     bMaterialIsStencil=false
     fontMaterial=null
     fontObject=&BurbankSmall
     letterSpacing=0
     monospacedWidth=1
     size=16
     skewAmount=0
     typefaceFontName=Black
     outlineSettings {
      bApplyOutlineToDropShadows=false
      bMiteredCorners=false
      bSeparateFillAlpha=false
      outlineMaterial=null
      outlineSize=0
      outlineColor=<rgba 0 0 0 1>
     }
    }
    shadowOffset=<0 0>
   }
   slot {class=BorderSlot, properties {horizontalAlignment=HAlign_Right, padding=<ltrb 0 0 0 0>}}
  }
  SizeBox_PlaceMarker {
   widgetClassPath=&SizeBox
   bInherited=true
   bIsVariable=true
   parent=GridPanel_buttons
   widget=SizeBox_PlaceMarker
   properties {bOverride_MaxDesiredWidth=true, maxDesiredWidth=305}
   slot {class=VerticalBoxSlot, properties {padding=<ltrb 0 0 0 8>}}
  }
  VerticalBox_152 {widgetClassPath=&VerticalBox, parent=SizeBox_PlaceMarker, widget=VerticalBox_152, slot {class=SizeBoxSlot}}
  HorizontalBox_144 {widgetClassPath=&HorizontalBox, bIsVariable=true, parent=VerticalBox_152, widget=HorizontalBox_144, slot {class=VerticalBoxSlot}}
  CommonHardwareVisibilityBorder_0 {
   widgetClassPath=&CommonHardwareVisibilityBorder
   parent=HorizontalBox_144
   widget=CommonHardwareVisibilityBorder_0
   properties {visibilityQuery {
     autoDescription=" ANY( Input.MouseAndKeyboard, Input.Gamepad )"
     tokenStreamVersion=0
     userDescription="K&&M-Controller"
     queryTokenStream [0,1,1,2,0,1]
     tagDictionary [{tagName=Input.MouseAndKeyboard}, {tagName=Input.Gamepad}]
    }}
   slot {class=HorizontalBoxSlot, properties {horizontalAlignment=HAlign_Right, verticalAlignment=VAlign_Center, size {sizeRule=Fill, value=1}}}
  }
  Text_PlaceMarker {
   widgetClassPath=&CommonTextBlock
   bInherited=true
   parent=CommonHardwareVisibilityBorder_0
   widget=Text_PlaceMarker
   properties {
    justification=Center
    scrollStyle=&TextScrollStyle-Base_C
    textTransformPolicy=ToUpper
    font {
     bForceMonospaced=false
     bMaterialIsStencil=false
     fontMaterial=null
     fontObject=&BurbankSmall
     letterSpacing=0
     monospacedWidth=1
     size=18
     skewAmount=0
     typefaceFontName=Black
     outlineSettings {
      bApplyOutlineToDropShadows=false
      bMiteredCorners=false
      bSeparateFillAlpha=false
      outlineMaterial=null
      outlineSize=0
      outlineColor=<rgba 0 0 0 0.25>
     }
    }
    shadowOffset=<0 0>
   }
   slot {class=BorderSlot, properties {padding=<ltrb 0 0 0 0>}}
  }
  CommonHardwareVisibilityBorder_1 {
   widgetClassPath=&CommonHardwareVisibilityBorder
   parent=HorizontalBox_144
   widget=CommonHardwareVisibilityBorder_1
   properties {visibilityQuery {
     autoDescription=" ANY( Input.Touch )"
     tokenStreamVersion=0
     userDescription=Touch
     queryTokenStream [0,1,1,1,0]
     tagDictionary [{tagName=Input.Touch}]
    }}
   slot {class=HorizontalBoxSlot, properties {horizontalAlignment=HAlign_Right, verticalAlignment=VAlign_Center, size {sizeRule=Fill, value=1}}}
  }
  Text_PlaceMarker_Mobile {
   widgetClassPath=&CommonTextBlock
   parent=CommonHardwareVisibilityBorder_1
   widget=Text_PlaceMarker_Mobile
   properties {
    justification=Center
    scrollStyle=&TextScrollStyle-Base_C
    text=Tap to place marker
    textTransformPolicy=ToUpper
    font {
     bForceMonospaced=false
     bMaterialIsStencil=false
     fontMaterial=null
     fontObject=&BurbankSmall
     letterSpacing=0
     monospacedWidth=1
     size=16
     skewAmount=0
     typefaceFontName=Black
     outlineSettings {
      bApplyOutlineToDropShadows=false
      bMiteredCorners=false
      bSeparateFillAlpha=false
      outlineMaterial=null
      outlineSize=0
      outlineColor=<rgba 0 0 0 1>
     }
    }
    shadowOffset=<0 0>
   }
   slot {class=BorderSlot, properties {padding=<ltrb 0 0 0 0>}}
  }
  Action_PlaceMarker {
   widgetClassPath=&CommonActionWidget
   bInherited=true
   parent=HorizontalBox_144
   widget=Action_PlaceMarker
   properties {visibility=SelfHitTestInvisible, inputActions [{dataTable=&AthenaMiniMapActionDataTable, rowName=ToggleMarker}]}
   slot {class=HorizontalBoxSlot, properties {padding=<ltrb 4 0 0 0>}}
  }
  Action_RemoveMarker {
   widgetClassPath=/Script/FortniteUI.ConcealableCommonActionWidget
   bInherited=true
   parent=HorizontalBox_144
   widget=Action_RemoveMarker
   properties {inputActions [{dataTable=&AthenaMiniMapActionDataTable, rowName=RemoveMarker}]}
   slot {class=HorizontalBoxSlot, properties {padding=<ltrb 4 0 0 0>}}
  }
  Panel_DangerPing {
   widgetClassPath=&HorizontalBox
   bInherited=true
   bIsVariable=true
   parent=VerticalBox_152
   widget=Panel_DangerPing
   slot {class=VerticalBoxSlot}
  }
  CommonHardwareVisibilityBorder {
   widgetClassPath=&CommonHardwareVisibilityBorder
   parent=Panel_DangerPing
   widget=CommonHardwareVisibilityBorder
   properties {visibilityQuery {
     autoDescription=" ANY( Input.MouseAndKeyboard, Input.Gamepad )"
     tokenStreamVersion=0
     userDescription="K&&M-Controller"
     queryTokenStream [0,1,1,2,0,1]
     tagDictionary [{tagName=Input.MouseAndKeyboard}, {tagName=Input.Gamepad}]
    }}
   slot {class=HorizontalBoxSlot, properties {horizontalAlignment=HAlign_Right, verticalAlignment=VAlign_Center, size {sizeRule=Fill, value=1}}}
  }
  Text_PlaceDangerMarker {
   widgetClassPath=&CommonTextBlock
   bInherited=true
   parent=CommonHardwareVisibilityBorder
   widget=Text_PlaceDangerMarker
   properties {
    justification=Center
    scrollStyle=&TextScrollStyle-Base_C
    textTransformPolicy=ToUpper
    font {
     bForceMonospaced=false
     bMaterialIsStencil=false
     fontMaterial=null
     fontObject=&BurbankSmall
     letterSpacing=0
     monospacedWidth=1
     size=18
     skewAmount=0
     typefaceFontName=Black
     outlineSettings {
      bApplyOutlineToDropShadows=false
      bMiteredCorners=false
      bSeparateFillAlpha=false
      outlineMaterial=null
      outlineSize=0
      outlineColor=<rgba 0 0 0 0.25>
     }
    }
    shadowOffset=<0 0>
   }
   slot {class=BorderSlot, properties {padding=<ltrb 0 0 0 0>}}
  }
  CommonHardwareVisibilityBorder_2 {
   widgetClassPath=&CommonHardwareVisibilityBorder
   parent=Panel_DangerPing
   widget=CommonHardwareVisibilityBorder_2
   properties {visibilityQuery {
     autoDescription=" ANY( Input.Touch )"
     tokenStreamVersion=0
     userDescription=Touch
     queryTokenStream [0,1,1,1,0]
     tagDictionary [{tagName=Input.Touch}]
    }}
   slot {class=HorizontalBoxSlot, properties {horizontalAlignment=HAlign_Right, verticalAlignment=VAlign_Center, size {sizeRule=Fill, value=1}}}
  }
  Text_PlaceMarker_Mobile_1 {
   widgetClassPath=&CommonTextBlock
   parent=CommonHardwareVisibilityBorder_2
   widget=Text_PlaceMarker_Mobile_1
   properties {
    justification=Center
    scrollStyle=&TextScrollStyle-Base_C
    text=Double tap to place danger
    textTransformPolicy=ToUpper
    font {
     bForceMonospaced=false
     bMaterialIsStencil=false
     fontMaterial=null
     fontObject=&BurbankSmall
     letterSpacing=0
     monospacedWidth=1
     size=16
     skewAmount=0
     typefaceFontName=Black
     outlineSettings {
      bApplyOutlineToDropShadows=false
      bMiteredCorners=false
      bSeparateFillAlpha=false
      outlineMaterial=null
      outlineSize=0
      outlineColor=<rgba 0 0 0 1>
     }
    }
    shadowOffset=<0 0>
   }
   slot {class=BorderSlot, properties {padding=<ltrb 0 0 0 0>}}
  }
  Action_PlaceMarkerDanger {
   widgetClassPath=&CommonActionWidget
   bInherited=true
   parent=Panel_DangerPing
   widget=Action_PlaceMarkerDanger
   properties {visibility=SelfHitTestInvisible, inputActions [{dataTable=&AthenaMiniMapActionDataTable, rowName=ToggleMarker}]}
   slot {class=HorizontalBoxSlot, properties {padding=<ltrb 4 0 0 0>}}
  }
  Action_RemoveMarkerDanger {
   widgetClassPath=&ConcealableCommonActionWidget
   bInherited=true
   parent=Panel_DangerPing
   widget=Action_RemoveMarkerDanger
   properties {inputActions [{dataTable=&AthenaMiniMapActionDataTable, rowName=RemoveMarker}]}
   slot {class=HorizontalBoxSlot, properties {padding=<ltrb 4 0 0 0>}}
  }
 ]
}
