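# Fullscreen map container widget holding the Athena map layer, challenge icons, browsing controls and loading spinner.
# When the minimap component opens the fullscreen map, it shows "Map Not Available" instead of map and markers if in play mode and the minigame's FNEUIComponent disables the map.
# Map-loading events swap the map for the spinner.
# Safe-zone side padding follows EnableMapSafeZone, and controls padding differs on mobile.
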WidgetBlueprint /FortniteCoreUI/HUD/MiniMap/AthenaMapLayerContainer
graphs [EventGraph {nodes [
   n1 = OnFullScreenMapToggle!() -> (OutputDelegate, then, bFullscreenMapVisible:bool=~) {is=title K2Node_CustomEvent}
   > Branch(Condition=@n1.bFullscreenMapVisible)
   n2 = GetOwningPlayer()
   > n3 = Cast To FortPlayerControllerAthena(Object=@n2)
   n4 = GetMinigame!(self:Fort Player Controller Gameplay Object Reference=@n3) -> (ReturnValue:Fort Minigame Object Reference)
   > n5 = Is Valid(InputObject=@n4)
   n6 = GetFNEUIComponent!(self:Fort Minigame Object Reference=@n4) -> (ReturnValue:FNEUIComponent Object Reference)
   > n7 = Is Valid(InputObject=@n6)
   n8 = IsInPlayMode!(Actor:"Actor Object Reference (by ref)"=@n2) -> (Play Mode:bool)
   > n9 = Branch(Condition=NOR Boolean!(A:bool=Get bShowMapWidget!(self:FNEUIComponent Object Reference=@n6) -> (bShowMapWidget:bool), B:bool=NOT Boolean(A=@n8)) -> (ReturnValue:bool))
   > SetVisibility(
    self=[Get ScaleBox_Map() -> (ScaleBox_Map:Scale Box Object Reference), Get SizeBox_MapMarkers() -> (SizeBox_MapMarkers:Size Box Object Reference)]
    InVisibility=Hidden
   )
   > SetVisibility(self=Get Border_Hidden() -> (Border_Hidden:Border Object Reference), InVisibility=HitTestInvisible)
   SetVisibility(
    execute=[@n9.else, @n7.Is Not Valid, @n5.Is Not Valid, @n3.CastFailed]
    self=[Get ScaleBox_Map(), Get SizeBox_MapMarkers()]
    InVisibility=Visible
   )
   > SetVisibility(self=Get Border_Hidden(), InVisibility=Hidden)
   Event Construct()
   > SetPadding!(
    execute
    self:Overlay Slot Object Reference=SlotAsOverlaySlot(Widget=Get SafeZone_Controls() -> (SafeZone_Controls:Safe Zone Object Reference))
    InPadding_Left:float=0
    InPadding_Top:float=0
    InPadding_Right:float=Select(Option 0:float=45, Option 1:float=25, Index=IsMobileGame()) -> (ReturnValue:float)
    InPadding_Bottom:float=120
   ) -> (then)
   > n10 = Cast To FortPlayerController(Object=GetOwningPlayer())
   n11 = GetMiniMapComponent!(self:Fort Player Controller Object Reference=@n10) -> (ReturnValue:Fort Controller Component Minimap Object Reference)
   > Is Valid(InputObject=@n11)
   > Bind Event to On Toggle Fullscreen Map(self:Fort Controller Component Minimap Object Reference=@n11, Delegate=Create Event()) {is=title K2Node_AddDelegate}
   Event PreConstruct()
   n12 = Get EnableMapSafeZone() -> (EnableMapSafeZone:bool)
   n13 = Get SafeZone_Map() -> (SafeZone_Map:Safe Zone Object Reference)
   > Set PadLeft!(execute, PadLeft:bool=@n12, self:Safe Zone Object Reference=@n13) -> (then, Output_Get:bool)
   > Set PadRight!(execute, PadRight:bool=@n12, self:Safe Zone Object Reference=@n13) -> (then, Output_Get:bool)
   Event Destruct()
   > n14 = Cast To FortPlayerController(Object=GetOwningPlayer())
   n15 = GetMiniMapComponent(self=@n14)
   > Is Valid(InputObject=@n15)
   > Unbind Event from On Toggle Fullscreen Map(self:Fort Controller Component Minimap Object Reference=@n15, Delegate=Create Event()) {is=title K2Node_RemoveDelegate}
   n16 = Event OnMapLoadingStateChanged!() -> (OutputDelegate, then, bIsLoading:bool, MapName:Name=None) {is=title K2Node_Event}
   > n17 = Branch(Condition=@n16.bIsLoading)
   > SetVisibility(self=Get Spinner_MapLoading() -> (Spinner_MapLoading:WBP UIKit Throbber Base Object Reference), InVisibility=Visible)
   > SetVisibility(self=Get ScaleBox_Map(), InVisibility=Collapsed)
   SetVisibility(execute=@n17.else, self=Get Spinner_MapLoading(), InVisibility=Collapsed)
   > SetVisibility(self=Get ScaleBox_Map(), InVisibility=SelfHitTestInvisible)
  ]}]
variables [EnableMapSafeZone {type=boolean, default=true}]
widgets {
 info {
  inheritedWidgetCount=5
  namedSlotCount=1
  parentClass=/Script/FortniteUI.AthenaMapLayerContainer
  rootWidgetClass=&Overlay
  widgetCount=13
 }
 widgets [
  Overlay_0 {widgetClassPath=&Overlay, widget=Overlay_0}
  SafeZone_Map {
   widgetClassPath=&SafeZone
   bIsVariable=true
   parent=Overlay_0
   widget=SafeZone_Map
   properties {padBottom=false, padTop=false}
   slot {class=OverlaySlot, properties {horizontalAlignment=HAlign_Fill, verticalAlignment=VAlign_Fill}}
  }
  ScaleBox_Map {widgetClassPath=&ScaleBox, bIsVariable=true, parent=SafeZone_Map, widget=ScaleBox_Map, slot {class=SafeZoneSlot}}
  AthenaMapLayer {
   widgetClassPath=/Script/FortniteUI.AthenaMapLayer
   bInherited=true
   parent=ScaleBox_Map
   widget=AthenaMapLayer
   properties {clipping=Inherit}
   slot {class=ScaleBoxSlot, properties {horizontalAlignment=HAlign_Fill, verticalAlignment=VAlign_Fill}}
  }
  AthenaMapIconsOverlay_Main {
   widgetClassPath=/Script/FortniteUI.AthenaInGameMapIconsOverlay
   bInherited=true
   namedSlotHost=AthenaMapLayer
   widget=AthenaMapIconsOverlay_Main
   properties {
    athenaMapNavigableIconSingleChallengeClass=&/AthenaMapNavigableIconSingleChallenge.AthenaMapNavigableIconSingleChallenge_C
    athenaMapNavigableIconStackOfChallengesClass=&/AthenaMapNavigableIconStackOfChallenges.AthenaMapNavigableIconStackOfChallenges_C
    athenaMapNavigableIconStackOfNoLocationChallengesClass=&/AthenaMapNavigableIconStackOfNoLocationChallenges.AthenaMapNavigableIconStackOfNoLocationChallenges_C
    bShowNoLocationIcon=false
   }
  }
  Spinner_MapLoading {
   widgetClassPath=/FortUILibrary/Systems/Throbber/WBP_UIKit_Throbber_Base
   bIsVariable=true
   parent=Overlay_0
   widget=Spinner_MapLoading
   properties {visibility=Collapsed}
   slot {class=OverlaySlot, properties {horizontalAlignment=HAlign_Center, verticalAlignment=VAlign_Center}}
  }
  SizeBox_MapMarkers {
   widgetClassPath=&SizeBox
   bIsVariable=true
   parent=Overlay_0
   widget=SizeBox_MapMarkers
   properties {bOverride_MaxDesiredWidth=true, maxDesiredWidth=800}
   slot {class=OverlaySlot, properties {horizontalAlignment=HAlign_Center, verticalAlignment=VAlign_Bottom}}
  }
  ListView_NonLocationIconsBelowMap {
   widgetClassPath=/Script/FortniteUI.AthenaMapChallengeListView
   bInherited=true
   parent=SizeBox_MapMarkers
   widget=ListView_NonLocationIconsBelowMap
   properties {
    clipping=Inherit
    entryWidgetClass=&/AthenaMapChallengeListEntry_Icon.AthenaMapChallengeListEntry_Icon_C
    numDesignerPreviewEntries=20
    orientation=Orient_Horizontal
   }
   slot {class=SizeBoxSlot}
  }
  SafeZone_Controls {
   widgetClassPath=&SafeZone
   bIsVariable=true
   parent=Overlay_0
   widget=SafeZone_Controls
   properties {padLeft=false, padTop=false}
   slot {class=OverlaySlot, properties {horizontalAlignment=HAlign_Right, verticalAlignment=VAlign_Bottom, padding=<ltrb 0 0 45 120>}}
  }
  MapBrowsingControlsPanel {
   widgetClassPath=&/BP_AthenaFullScreenMapBrowsingControlsWidget
   bInherited=true
   bIsVariable=true
   parent=SafeZone_Controls
   widget=MapBrowsingControlsPanel
   properties {bSetVisibilityOnActivated=true, bSetVisibilityOnDeactivated=true}
   slot {class=SafeZoneSlot}
  }
  Border_Hidden {
   widgetClassPath=&Border
   bIsVariable=true
   parent=Overlay_0
   widget=Border_Hidden
   properties {horizontalAlignment=HAlign_Center, verticalAlignment=VAlign_Center, visibility=Hidden, brushColor=<rgba 0 0 0 0>}
   slot {class=OverlaySlot, properties {horizontalAlignment=HAlign_Fill, verticalAlignment=VAlign_Fill}}
  }
  CommonTextBlock_0 {
   widgetClassPath=&CommonTextBlock
   parent=Border_Hidden
   widget=CommonTextBlock_0
   properties {
    text=Map Not Available
    font {
     bForceMonospaced=false
     bMaterialIsStencil=false
     fontMaterial=null
     fontObject=&BurbankSmall
     letterSpacing=0
     monospacedWidth=1
     size=24
     skewAmount=0
     typefaceFontName=Bold
     outlineSettings {
      bApplyOutlineToDropShadows=false
      bMiteredCorners=false
      bSeparateFillAlpha=false
      outlineMaterial=null
      outlineSize=0
      outlineColor=<rgba 0 0 0 1>
     }
    }
    shadowOffset=<0 0>
   }
   slot {class=BorderSlot, properties {horizontalAlignment=HAlign_Center, verticalAlignment=VAlign_Center}}
  }
  Text_CreativeTeleportCountdown {
   widgetClassPath=&CommonTextBlock
   bInherited=true
   parent=Overlay_0
   widget=Text_CreativeTeleportCountdown
   properties {
    text=Teleporting in
    visibility=Collapsed
    font {
     bForceMonospaced=false
     bMaterialIsStencil=false
     fontMaterial=null
     fontObject=&BurbankSmall
     letterSpacing=0
     monospacedWidth=1
     size=24
     skewAmount=0
     typefaceFontName=Bold
     outlineSettings {
      bApplyOutlineToDropShadows=false
      bMiteredCorners=false
      bSeparateFillAlpha=false
      outlineMaterial=null
      outlineSize=0
      outlineColor=<rgba 0 0 0 1>
     }
    }
    shadowOffset=<0 0>
   }
   slot {class=OverlaySlot, properties {horizontalAlignment=HAlign_Center, verticalAlignment=VAlign_Center}}
  }
 ]
}
