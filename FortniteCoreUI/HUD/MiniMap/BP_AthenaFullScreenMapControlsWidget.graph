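# Fullscreen map controls widget for Athena, with Reset, Exit and a default-collapsed Show Quest Icons button.
# Binds to the minimap component's fullscreen toggle; on opening, hides the Reset row if play mode is active and the minigame UI disables the map widget.
# Input type changes show button backers only for gamepad, though the Reset backer stays collapsed once collapsed.
# Visibility wrappers hide Reset and Exit on touch.
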WidgetBlueprint /FortniteCoreUI/HUD/MiniMap/BP_AthenaFullScreenMapControlsWidget
graphs [
 EventGraph {nodes [
   n1 = OnFullScreenMapToggle!() -> (OutputDelegate, then, bFullscreenMapVisible:bool=~) {is=title K2Node_CustomEvent}
   > Branch(Condition=@n1.bFullscreenMapVisible)
   n2 = GetOwningPlayer()
   > n3 = Cast To FortPlayerControllerAthena(Object=@n2)
   n4 = GetMinigame!(self:Fort Player Controller Gameplay Object Reference=@n3) -> (ReturnValue:Fort Minigame Object Reference)
   > n5 = Is Valid(InputObject=@n4)
   n6 = GetFNEUIComponent!(self:Fort Minigame Object Reference=@n4) -> (ReturnValue:FNEUIComponent Object Reference)
   > n7 = Is Valid(InputObject=@n6)
   n8 = IsInPlayMode!(Actor:"Actor Object Reference (by ref)"=@n2) -> (Play Mode:bool)
   > n9 = Branch(Condition=NOR Boolean!(A:bool=Get bShowMapWidget!(self:FNEUIComponent Object Reference=@n6) -> (bShowMapWidget:bool), B:bool=NOT Boolean(A=@n8)) -> (ReturnValue:bool))
   > SetVisibility(self=Get HorizontalBox_Controls() -> (HorizontalBox_Controls:Horizontal Box Object Reference), InVisibility=Hidden)
   SetVisibility(execute=[@n9.else, @n7.Is Not Valid, @n5.Is Not Valid, @n3.CastFailed], self=Get HorizontalBox_Controls(), InVisibility=Visible)
   Event Construct()
   > UpdateAccordingInputType!(execute, self, InputType:ECommonInputType Enum=GetCurrentInputType(self=Get CommonInputSubsystem())) -> (then)
   > n10 = Cast To FortPlayerController(Object=GetOwningPlayer())
   n11 = GetMiniMapComponent!(self:Fort Player Controller Object Reference=@n10) -> (ReturnValue:Fort Controller Component Minimap Object Reference)
   > Is Valid(InputObject=@n11)
   > Bind Event to On Toggle Fullscreen Map(self:Fort Controller Component Minimap Object Reference=@n11, Delegate=Create Event()) {is=title K2Node_AddDelegate}
   Event Destruct()
   > n12 = Cast To FortPlayerController(Object=GetOwningPlayer())
   n13 = GetMiniMapComponent(self=@n12)
   > Is Valid(InputObject=@n13)
   > Unbind Event from On Toggle Fullscreen Map(self:Fort Controller Component Minimap Object Reference=@n13, Delegate=Create Event()) {is=title K2Node_RemoveDelegate}
   n14 = Event On Input Type Changed!() -> (OutputDelegate, then, NewInputType:ECommonInputType Enum) {is=title K2Node_Event}
   > UpdateAccordingInputType(InputType=@n14.NewInputType)
  ]}
 UpdateAccordingInputType {nodes [
   n1 = UpdateAccordingInputType!() -> (then, InputType:ECommonInputType Enum) {is=title K2Node_FunctionEntry}
   n2 = "Equal (Enum)"(A=@n1, B=Gamepad)
   n3 = "Not Equal (Enum)"!(
    A:ESlateVisibility Enum=GetVisibility!(self:Widget Object Reference=Get GamePadButtonResetBacker() -> (GamePadButtonResetBacker:Image Object Reference)) -> (ReturnValue:ESlateVisibility Enum=Visible)
    B:ESlateVisibility Enum=Collapsed
   ) -> (ReturnValue:bool)
   > SetVisibility(self=Get GamePadButtonResetBacker(), InVisibility=Select(Option 0=Collapsed, Option 1=Visible, Index=AND Boolean(A=@n2, B=@n3)))
   > SetVisibility(
    self=Get GamePadButtonExitBacker() -> (GamePadButtonExitBacker:Image Object Reference)
    InVisibility=Select(Option 0=Collapsed, Option 1=Visible, Index=@n2)
   )
  ]}
]
variables []
widgets {
 info {
  inheritedWidgetCount=3
  parentClass=/Script/FortniteUI.AthenaFullScreenMapControlsWidget
  rootWidgetClass=&Overlay
  widgetCount=15
 }
 widgets [
  Overlay_1 {widgetClassPath=&Overlay, widget=Overlay_1}
  WBP_CaptureForPostBufferUpdate {
   widgetClassPath=&WBP_CaptureForPostBufferUpdate
   bIsVariable=true
   parent=Overlay_1
   slot {class=OverlaySlot, properties {horizontalAlignment=HAlign_Fill, verticalAlignment=VAlign_Fill}}
  }
  HorizontalBox_2 {
   widgetClassPath=&HorizontalBox
   parent=Overlay_1
   widget=HorizontalBox_2
   slot {class=OverlaySlot, properties {verticalAlignment=VAlign_Center}}
  }
  SizeBox_ToggleQuestIcons {
   widgetClassPath=&SizeBox
   bIsVariable=true
   parent=HorizontalBox_2
   widget=SizeBox_ToggleQuestIcons
   properties {bOverride_MaxDesiredWidth=true, maxDesiredWidth=305}
   slot {class=HorizontalBoxSlot}
  }
  Button_ToggleQuestIcons {
   widgetClassPath=/FortUILibrary/UIKit/Buttons/WBP_UIKit_Button_Quiet
   bInherited=true
   bIsVariable=true
   parent=SizeBox_ToggleQuestIcons
   widget=Button_ToggleQuestIcons
   properties {
    minWidth=120
    text=Show Quest Icons
    visibility=Collapsed
    triggeringInputAction {dataTable=/FortUILibrary/Input/AthenaMiniMapActionDataTable, rowName=ShowQuestIcons}
   }
   slot {class=SizeBoxSlot, properties {horizontalAlignment=HAlign_Right, verticalAlignment=VAlign_Center, padding=<ltrb 24 0 0 0>}}
  }
  HorizontalBox_Controls {
   widgetClassPath=&HorizontalBox
   bIsVariable=true
   parent=Overlay_1
   widget=HorizontalBox_Controls
   slot {class=OverlaySlot, properties {horizontalAlignment=HAlign_Center, verticalAlignment=VAlign_Center}}
  }
  CommonVisibilityWidget_2 {
   widgetClassPath=&CommonVisibilityWidgetLegacy
   parent=HorizontalBox_Controls
   widget=CommonVisibilityWidget_2
   properties {bShowForTouch=false, padding=<ltrb 0 0 0 0>}
   slot {class=HorizontalBoxSlot, properties {padding=<ltrb 15 0 0 0>}}
  }
  SizeBox_12 {
   widgetClassPath=&SizeBox
   parent=CommonVisibilityWidget_2
   widget=SizeBox_12
   properties {bOverride_MaxDesiredWidth=true, maxDesiredWidth=305}
   slot {class=BorderSlot, properties {padding=<ltrb 0 0 0 0>}}
  }
  Overlay_2 {
   widgetClassPath=&Overlay
   parent=SizeBox_12
   widget=Overlay_2
   slot {class=SizeBoxSlot, properties {horizontalAlignment=HAlign_Right, verticalAlignment=VAlign_Center, padding=<ltrb 32 0 0 0>}}
  }
  GamePadButtonResetBacker {
   widgetClassPath=&Image
   bIsVariable=true
   parent=Overlay_2
   widget=GamePadButtonResetBacker
   properties {
    renderOpacity=0.8
    visibility=HitTestInvisible
    brush {
     bIsDynamicallyLoaded=false
     drawAs=Image
     imageType=NoImage
     mirroring=NoMirror
     resourceName=None
     resourceObject=/FortniteCoreUI/UI/DiscoveryShared/Art/MI_UI_Discovery_RoundedCornersRectangle_Join
     tiling=NoTile
     imageSize=<32 32>
     margin=<ltrb 0 0 0 0>
     outlineSettings {
      bUseBrushTransparency=false
      roundingType=HalfHeightRadius
      width=0
      color {colorUseRule=UseColor_Specified, specifiedColor=<rgba 0 0 0 0>}
      cornerRadii=<0 0 0 0>
     }
     tintColor {colorUseRule=UseColor_Specified, specifiedColor=<rgba 1 1 1 1>}
     uVRegion {bIsValid=false, max=<0 0>, min=<0 0>}
    }
    colorAndOpacity=<rgba 0.0296 0.0296 0.0356 1>
   }
   slot {class=OverlaySlot, properties {horizontalAlignment=HAlign_Fill, verticalAlignment=VAlign_Fill, padding=<ltrb -32 1 0 1>}}
  }
  Button_ResetToPlayer {
   widgetClassPath=&WBP_UIKit_Button_Quiet
   bInherited=true
   bIsVariable=true
   parent=Overlay_2
   widget=Button_ResetToPlayer
   properties {text=Reset, triggeringInputAction {dataTable=&AthenaMiniMapActionDataTable, rowName=ResetToPlayer}}
   slot {class=OverlaySlot}
  }
  HardwareVisibilityBorder_Exit {
   widgetClassPath=&CommonHardwareVisibilityBorder
   parent=Overlay_1
   widget=HardwareVisibilityBorder_Exit
   properties {visibilityQuery {
     autoDescription=" NONE( Input.Touch )"
     tokenStreamVersion=0
     userDescription=""
     queryTokenStream [0,1,3,1,0]
     tagDictionary [{tagName=Input.Touch}]
    }}
   slot {class=OverlaySlot, properties {horizontalAlignment=HAlign_Right, verticalAlignment=VAlign_Center, padding=<ltrb 32 0 0 0>}}
  }
  Overlay_0 {
   widgetClassPath=&Overlay
   parent=HardwareVisibilityBorder_Exit
   widget=Overlay_0
   slot {class=BorderSlot, properties {padding=<ltrb 0 0 0 0>}}
  }
  GamePadButtonExitBacker {
   widgetClassPath=&Image
   bIsVariable=true
   parent=Overlay_0
   widget=GamePadButtonExitBacker
   properties {
    renderOpacity=0.8
    visibility=HitTestInvisible
    brush {
     bIsDynamicallyLoaded=false
     drawAs=Image
     imageType=NoImage
     mirroring=NoMirror
     resourceName=None
     resourceObject=&MI_UI_Discovery_RoundedCornersRectangle_Join
     tiling=NoTile
     imageSize=<32 32>
     margin=<ltrb 0 0 0 0>
     outlineSettings {
      bUseBrushTransparency=false
      roundingType=HalfHeightRadius
      width=0
      color {colorUseRule=UseColor_Specified, specifiedColor=<rgba 0 0 0 0>}
      cornerRadii=<0 0 0 0>
     }
     tintColor {colorUseRule=UseColor_Specified, specifiedColor=<rgba 1 1 1 1>}
     uVRegion {bIsValid=false, max=<0 0>, min=<0 0>}
    }
    colorAndOpacity=<rgba 0.0296 0.0296 0.0356 1>
   }
   slot {class=OverlaySlot, properties {horizontalAlignment=HAlign_Fill, verticalAlignment=VAlign_Fill, padding=<ltrb 0 1 0 1>}}
  }
  Button_Exit {
   widgetClassPath=&WBP_UIKit_Button_Quiet
   bInherited=true
   bIsVariable=true
   parent=Overlay_0
   widget=Button_Exit
   properties {minWidth=120, text=Exit, triggeringInputAction {dataTable=&InputActionDataTable, rowName=GenericBack}}
   slot {class=OverlaySlot}
  }
 ]
}
