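# Widget Blueprint for the full-screen map panel, layering a backing gradient, map layers, bottom controls, game info slot and legend.
# On activation outside a zone, it plays an Intro that fades and slides the controls in.
# It listens on the playspace event router for map controller device events, collapsing game info and gradient when a custom map hides the info panel.
# The game info widget fills its overlay slot once created.
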WidgetBlueprint /FortniteCoreUI/HUD/MiniMap/BP_AthenaFullScreenMapPanel
animations [Intro {
  binding_count=1
  length_seconds=0.1
  object=&self:Intro
  sequence {bindings [BP_AthenaFullScreenMapControlsWidget {
     id=#1
     possessed_class=BP_AthenaFullScreenMapControlsWidget_C
     tracks [
      {
       type=MovieSceneFloatTrack
       display_name=Render Opacity
       sections [{type=MovieSceneFloatSection, channels [None {type=MovieSceneScriptingFloatChannel, keys [[0,0],[2,1]]}]}]
      }
      {
       type=MovieScene2DTransformTrack
       sections [{
         type=MovieScene2DTransformSection
         channels [
          Translation.X {type=MovieSceneScriptingFloatChannel, const [0,1]}
          Translation.Y {type=MovieSceneScriptingFloatChannel, keys [[0,16],[2,0]]}
          Angle {type=MovieSceneScriptingFloatChannel}
          Scale.X {type=MovieSceneScriptingFloatChannel}
          Scale.Y {type=MovieSceneScriptingFloatChannel}
          Shear.X {type=MovieSceneScriptingFloatChannel}
          Shear.Y {type=MovieSceneScriptingFloatChannel}
         ]
        }]
      }
     ]
    }]}
 }]
graphs [EventGraph {nodes [
   Event On Activated()
   > n1 = Branch(Condition=IsInZone!(self:Fort Global UIContext Object Reference=Get FortGlobalUIContext() -> (ReturnValue:Fort Global UIContext Object Reference) {is=title K2Node_GetSubsystem}) -> (ReturnValue:bool))
   PlayAnimation(execute=@n1.else, InAnimation=Get Intro())
   Event Destruct()
   > StopListeningToEvent(self:Async Action Start Listening To Event Object Reference=Get EventListener_MapControllerDeviceActivated() -> (EventListener_MapControllerDeviceActivated:Async Action Start Listening To Event Object Reference))
   Event Construct()
   > n2 = GetEventRouter!(execute, Scope:Object Class Reference=/Script/FortniteGame.FortPlayspace, ContextActor:Actor Object Reference=GetOwningPlayer()) -> (then, ReturnValue:Gameplay Event Router Component Object Reference)
   > Is Valid(InputObject=@n2)
   > n3 = StartListeningToEvent!(
    execute:Exec
    Target:Gameplay Event Router Component Object Reference=@n2
    EventType:Script Struct Object Reference=/Game/Athena/HUD/MiniMap/MapController/GER_MapControllerDevice_ActivePayload
    Context:Object Reference
    EventBubblingRule:EEventBubblingRule Enum
   ) -> (
    then:Exec
    AsyncTaskProxy:Async Action Start Listening To Event Object Reference
    OnEventReceived:Exec
    Payload_CustomMapActive_1_#2:bool=~
    Payload_HideMapInfoPanel_4_#3:bool=~
    Payload_TransparentMinimap_12_#4:bool=~
    Payload_MapShape_10_#5:EMapController_MapShape Enum
    Payload_MapRotationMode_16_#6:EMapControllerDevice_RotationMode Enum
   )
   > Set EventListener_MapControllerDeviceActivated(EventListener_MapControllerDeviceActivated:Async Action Start Listening To Event Object Reference=@n3) -> (Output_Get:Async Action Start Listening To Event Object Reference)
   SetVisibility(
    execute=@n3.OnEventReceived
    self=[
     Get Overlay_GameInfoSlot() -> (Overlay_GameInfoSlot:Overlay Object Reference)
     Get Image_MapFalloff() -> (Image_MapFalloff:Image Object Reference)
    ]
    InVisibility=Select(
     Option 0=Visible
     Option 1=Collapsed
     Index=AND Boolean(A=@n3.Payload_CustomMapActive_1_#2, B=@n3.Payload_HideMapInfoPanel_4_#3)
    )
   )
   n4 = Event OnMapGameInfoWidgetCreated!() -> (OutputDelegate, then, Widget:User Widget Object Reference) {is=title K2Node_Event}
   > n5 = AddChildToOverlay!(execute, self:Overlay Object Reference=Get Overlay_GameInfoSlot(), Content:Widget Object Reference=@n4.Widget) -> (then, ReturnValue:Overlay Slot Object Reference)
   > SetHorizontalAlignment(self=@n5, InHorizontalAlignment=HAlign_Fill)
   > SetVerticalAlignment(self=@n5, InVerticalAlignment=VAlign_Fill)
  ]}]
variables [EventListener_MapControllerDeviceActivated {type=AsyncAction_StartListeningToEvent, default=null}]
widgets {
 info {inheritedWidgetCount=2, parentClass=/Script/FortniteUI.AthenaFullScreenMapPanel, rootWidgetClass=&Overlay, widgetCount=8}
 widgets [
  Overlay_0 {widgetClassPath=&Overlay, widget=Overlay_0}
  Image_MapFalloff {
   widgetClassPath=&Image
   bIsVariable=true
   parent=Overlay_0
   widget=Image_MapFalloff
   properties {
    visibility=SelfHitTestInvisible
    brush {
     bIsDynamicallyLoaded=false
     drawAs=Image
     imageType=NoImage
     mirroring=NoMirror
     resourceName=None
     resourceObject=/Game/Athena/Apollo/Maps/UI/M_MapBacking_Gradient
     tiling=NoTile
     imageSize=<32 32>
     margin=<ltrb 0 0 0 0>
     outlineSettings {
      bUseBrushTransparency=false
      roundingType=HalfHeightRadius
      width=0
      color {colorUseRule=UseColor_Specified, specifiedColor=<rgba 0 0 0 0>}
      cornerRadii=<0 0 0 0>
     }
     tintColor {colorUseRule=UseColor_Specified, specifiedColor=<rgba 1 1 1 1>}
     uVRegion {bIsValid=false, max=<0 0>, min=<0 0>}
    }
   }
   slot {class=OverlaySlot, properties {horizontalAlignment=HAlign_Fill, verticalAlignment=VAlign_Fill, padding=<ltrb -2 -2 -2 -2>}}
  }
  MapLayerContainer {
   widgetClassPath=&/AthenaMapLayerContainer
   bInherited=true
   parent=Overlay_0
   widget=MapLayerContainer
   properties {bIsFocusable=true, bSupportsActivationFocus=true}
   slot {class=OverlaySlot, properties {horizontalAlignment=HAlign_Fill, verticalAlignment=VAlign_Fill}}
  }
  SafeZone_169 {
   widgetClassPath=&SafeZone
   parent=Overlay_0
   widget=SafeZone_169
   properties {padTop=false}
   slot {class=OverlaySlot, properties {horizontalAlignment=HAlign_Fill, verticalAlignment=VAlign_Bottom}}
  }
  BP_AthenaFullScreenMapControlsWidget {
   widgetClassPath=&/BP_AthenaFullScreenMapControlsWidget
   parent=SafeZone_169
   widget=BP_AthenaFullScreenMapControlsWidget
   slot {class=SafeZoneSlot, properties {padding=<ltrb 40 0 40 16>}}
  }
  Overlay_GameInfoSlot {
   widgetClassPath=&Overlay
   bIsVariable=true
   parent=Overlay_0
   widget=Overlay_GameInfoSlot
   slot {class=OverlaySlot, properties {padding=<ltrb 56 112 0 64>}}
  }
  SafeZone_Legends {
   widgetClassPath=&SafeZone
   parent=Overlay_0
   widget=SafeZone_Legends
   slot {class=OverlaySlot, properties {horizontalAlignment=HAlign_Right, padding=<ltrb 0 110 45 0>}}
  }
  MapLegend {
   widgetClassPath=&/BP_AthenaMapLegendsWidget
   bInherited=true
   parent=SafeZone_Legends
   widget=MapLegend
   slot {class=SafeZoneSlot}
  }
 ]
}
