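# Widget Blueprint for one legend entry on the Athena map, showing a name beside one or more icons, built on the native AthenaMapLegendEntryWidget.
# A horizontal box holds a width-capped, wrapping name text and an overlay containing a dynamic entry box of WBP_LegendIcon widgets.
# On PreConstruct, SetMobileStyle checks for compact UI, then enlarges the icon overlay and widens the name's slot padding.
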WidgetBlueprint /FortniteCoreUI/HUD/MiniMap/BP_AthenaMapLegendEntryWidget
graphs [
 EventGraph {nodes [
   Event PreConstruct()
   > SetMobileStyle() {is=title K2Node_CallFunction}
   Event Construct()
  ]}
 SetMobileStyle {nodes [
   SetMobileStyle() {is=title K2Node_FunctionEntry}
   > Branch(Condition=IsUsingCompactUI())
   > SetRenderScale!(
    execute
    self:Widget Object Reference=Get Overlay_LegendIcons() -> (Overlay_LegendIcons:Overlay Object Reference)
    Scale:Vector 2D Structure=(X=1.35,Y=1.35)
   ) -> (then)
   > Set!(
    execute
    Padding_Left:float=8
    Padding_Top:float=10
    Padding_Right:float=15
    Padding_Bottom:float=10
    self:Horizontal Box Slot Object Reference=SlotAsHorizontalBoxSlot!(Widget:Widget Object Reference=Get SizeBox_EntryName() -> (SizeBox_EntryName:Size Box Object Reference)) -> (ReturnValue:Horizontal Box Slot Object Reference)
   ) -> (then, Output_Get:Margin Structure)
  ]}
]
variables []
widgets {
 info {inheritedWidgetCount=2, parentClass=/Script/FortniteUI.AthenaMapLegendEntryWidget, rootWidgetClass=&HorizontalBox, widgetCount=5}
 widgets [
  HorizontalBox_EntryWidget {widgetClassPath=&HorizontalBox, bIsVariable=true, widget=HorizontalBox_EntryWidget}
  SizeBox_EntryName {
   widgetClassPath=&SizeBox
   bIsVariable=true
   parent=HorizontalBox_EntryWidget
   widget=SizeBox_EntryName
   properties {bOverride_MaxDesiredWidth=true, maxDesiredWidth=340}
   slot {
    class=HorizontalBoxSlot
    properties {horizontalAlignment=HAlign_Left, verticalAlignment=VAlign_Center, padding=<ltrb 8 0 0 0>, size {sizeRule=Fill, value=1}}
   }
  }
  Text_EntryName {
   widgetClassPath=&CommonTextBlock
   bInherited=true
   bIsVariable=true
   parent=SizeBox_EntryName
   widget=Text_EntryName
   properties {
    mobileFontSizeMultiplier=1.3
    text=$Legend Entry
    wrapTextAt=300
    font {
     bForceMonospaced=false
     bMaterialIsStencil=false
     fontMaterial=null
     fontObject=&BurbankSmall
     letterSpacing=0
     monospacedWidth=1
     size=15
     skewAmount=0
     typefaceFontName=Bold
     outlineSettings {
      bApplyOutlineToDropShadows=false
      bMiteredCorners=false
      bSeparateFillAlpha=false
      outlineMaterial=null
      outlineSize=0
      outlineColor=<rgba 0 0 0 1>
     }
    }
    shadowOffset=<0 0>
   }
   slot {class=SizeBoxSlot}
  }
  Overlay_LegendIcons {
   widgetClassPath=&Overlay
   bIsVariable=true
   parent=HorizontalBox_EntryWidget
   widget=Overlay_LegendIcons
   slot {class=HorizontalBoxSlot, properties {horizontalAlignment=HAlign_Left, verticalAlignment=VAlign_Center}}
  }
  DynamicEntryBox_LegendIcons {
   widgetClassPath=&DynamicEntryBox
   bInherited=true
   bIsVariable=true
   parent=Overlay_LegendIcons
   widget=DynamicEntryBox_LegendIcons
   properties {entryWidgetClass=&/WBP_LegendIcon.WBP_LegendIcon_C, numDesignerPreviewEntries=1}
   slot {class=OverlaySlot}
  }
 ]
}
