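# Minimap HUD widget that shows how many players, and squads, are left in an Athena match.
# It subclasses AthenaMapPlayersLeftWidget, laying out an auto-wrapping title beside the inherited count text inside a width-capped SizeBox.
# When the max squad size is set, the title switches to "Players (Squads) Left" if that size exceeds one, otherwise "Players Left".
# The graph only sets the title and leaves the count value untouched.
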WidgetBlueprint /FortniteCoreUI/HUD/MiniMap/BP_AthenaMapPlayersLeft
graphs [EventGraph {nodes [
   n1 = Event OnMaxSquadSizeSet!() -> (OutputDelegate, then, MaxSquadSize:int=0) {is=title K2Node_Event}
   > "SetText (Text)"(
    self=Get TitleText() -> (TitleText:Common Text Object Reference)
    InText=Select(
     Option 0:Text=Get PlayersLeft() -> (PlayersLeft:Text)
     Option 1:Text=Get PlayersSquadsLeft() -> (PlayersSquadsLeft:Text)
     Index="integer > integer"(A=@n1.MaxSquadSize, B=1) {is=title K2Node_CallFunction}
    ) -> (ReturnValue:Text)
   )
  ]}]
variables [
 PlayersLeft {type=string, default=Players Left}
 PlayersSquadsLeft {type=string, default="Players (Squads) Left"}
]
widgets {
 info {inheritedWidgetCount=2, parentClass=/Script/FortniteUI.AthenaMapPlayersLeftWidget, rootWidgetClass=&SizeBox, widgetCount=5}
 widgets [
  SizeBox_0 {widgetClassPath=&SizeBox, widget=SizeBox_0, properties {bOverride_MaxDesiredWidth=true, maxDesiredWidth=336}}
  HorizontalBox_1 {
   widgetClassPath=&HorizontalBox
   parent=SizeBox_0
   widget=HorizontalBox_1
   properties {bOverride_Cursor=true}
   slot {class=SizeBoxSlot}
  }
  TitleText {
   widgetClassPath=&CommonTextBlock
   bIsVariable=true
   parent=HorizontalBox_1
   widget=TitleText
   properties {
    autoWrapText=true
    text="{PLAYERS (SQUADS) LEFT}"
    textTransformPolicy=ToUpper
    colorAndOpacity {colorUseRule=UseColor_Specified, specifiedColor=<rgba 0.2122 0.6514 1 1>}
    font {
     bForceMonospaced=false
     bMaterialIsStencil=false
     fontMaterial=null
     fontObject=&BurbankSmall
     letterSpacing=0
     monospacedWidth=1
     size=15
     skewAmount=0
     typefaceFontName=Bold
     outlineSettings {
      bApplyOutlineToDropShadows=false
      bMiteredCorners=false
      bSeparateFillAlpha=false
      outlineMaterial=null
      outlineSize=0
      outlineColor=<rgba 0 0 0 1>
     }
    }
    shadowOffset=<0 0>
   }
   slot {
    class=HorizontalBoxSlot
    properties {horizontalAlignment=HAlign_Left, verticalAlignment=VAlign_Center, padding=<ltrb 0 2 0 0>, size {sizeRule=Fill, value=1}}
   }
  }
  Text_PlayersTeamsLeft {
   widgetClassPath=&CommonTextBlock
   bInherited=true
   parent=HorizontalBox_1
   widget=Text_PlayersTeamsLeft
   properties {
    text="{0 (0)}"
    font {
     bForceMonospaced=false
     bMaterialIsStencil=false
     fontMaterial=null
     fontObject=&BurbankSmall
     letterSpacing=0
     monospacedWidth=1
     size=21
     skewAmount=0
     typefaceFontName=Bold
     outlineSettings {
      bApplyOutlineToDropShadows=false
      bMiteredCorners=false
      bSeparateFillAlpha=false
      outlineMaterial=null
      outlineSize=0
      outlineColor=<rgba 0 0 0 1>
     }
    }
    shadowOffset=<0 0>
   }
   slot {class=HorizontalBoxSlot, properties {horizontalAlignment=HAlign_Right, verticalAlignment=VAlign_Top, padding=<ltrb 12 0 0 0>}}
  }
  TutorialHighlight {widgetClassPath=&TutorialHighlight}
 ]
}
