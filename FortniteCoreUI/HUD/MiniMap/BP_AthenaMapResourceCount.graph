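# Widget Blueprint for a resource count row in the Athena minimap HUD, showing a resource name, icon and count.
# A horizontal box puts the uppercase light-blue name in the filling left slot and a nested, right-aligned box with the lazy-loaded icon and larger count.
# The event graph is empty; the Blueprint only lays out and styles widgets inherited from FortResourceCountWidget, with placeholder text for name and count.
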WidgetBlueprint /FortniteCoreUI/HUD/MiniMap/BP_AthenaMapResourceCount
graphs [EventGraph {links [], nodes []}]
variables []
widgets {
 info {inheritedWidgetCount=4, parentClass=/Script/FortniteUI.FortResourceCountWidget, rootWidgetClass=&HorizontalBox, widgetCount=6}
 widgets [
  HorizontalBox_1 {widgetClassPath=&HorizontalBox, widget=HorizontalBox_1, properties {bOverride_Cursor=true}}
  Text_ResourceName {
   widgetClassPath=&CommonTextBlock
   bInherited=true
   parent=HorizontalBox_1
   widget=Text_ResourceName
   properties {
    text="{GOLD}"
    textTransformPolicy=ToUpper
    colorAndOpacity {colorUseRule=UseColor_Specified, specifiedColor=<rgba 0.2122 0.6514 1 1>}
    font {
     bForceMonospaced=false
     bMaterialIsStencil=false
     fontMaterial=null
     fontObject=&BurbankSmall
     letterSpacing=0
     monospacedWidth=1
     size=15
     skewAmount=0
     typefaceFontName=Bold
     outlineSettings {
      bApplyOutlineToDropShadows=false
      bMiteredCorners=false
      bSeparateFillAlpha=false
      outlineMaterial=null
      outlineSize=0
      outlineColor=<rgba 0 0 0 1>
     }
    }
    shadowOffset=<0 0>
   }
   slot {class=HorizontalBoxSlot, properties {horizontalAlignment=HAlign_Left, verticalAlignment=VAlign_Center, size {sizeRule=Fill, value=1}}}
  }
  HorizontalBox_0 {
   widgetClassPath=&HorizontalBox
   parent=HorizontalBox_1
   widget=HorizontalBox_0
   slot {class=HorizontalBoxSlot, properties {horizontalAlignment=HAlign_Right}}
  }
  LazyImage_ResourceIcon {
   widgetClassPath=&FortLazyImage
   bInherited=true
   parent=HorizontalBox_0
   widget=LazyImage_ResourceIcon
   slot {class=HorizontalBoxSlot}
  }
  Text_ResourceCount {
   widgetClassPath=&CommonTextBlock
   bInherited=true
   parent=HorizontalBox_0
   widget=Text_ResourceCount
   properties {
    text="{0}"
    font {
     bForceMonospaced=false
     bMaterialIsStencil=false
     fontMaterial=null
     fontObject=&BurbankSmall
     letterSpacing=0
     monospacedWidth=1
     size=21
     skewAmount=0
     typefaceFontName=Bold
     outlineSettings {
      bApplyOutlineToDropShadows=false
      bMiteredCorners=false
      bSeparateFillAlpha=false
      outlineMaterial=null
      outlineSize=0
      outlineColor=<rgba 0 0 0 1>
     }
    }
    shadowOffset=<0 0>
   }
   slot {class=HorizontalBoxSlot, properties {padding=<ltrb 8 0 0 0>}}
  }
  TutorialHighlight {widgetClassPath=&TutorialHighlight}
 ]
}
