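# Widget Blueprint for the squad eliminations panel on the Athena map, built on the native AthenaMapSquadEliminationsWidget.
# A vertical box stacks an uppercase "Eliminations" header, a player eliminations row flagged as the local player, and a vertical dynamic entry box.
# The inherited entry box uses the same player eliminations widget as its entry class for squad rows.
# The Blueprint has no graph logic or variables, only layout and styling.
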WidgetBlueprint /FortniteCoreUI/HUD/MiniMap/BP_AthenaMapSquadEliminations
graphs [EventGraph {links [], nodes []}]
variables []
widgets {
 info {
  inheritedWidgetCount=2
  parentClass=/Script/FortniteUI.AthenaMapSquadEliminationsWidget
  rootWidgetClass=&VerticalBox
  widgetCount=5
 }
 widgets [
  VerticalBox_26 {widgetClassPath=&VerticalBox, widget=VerticalBox_26}
  CommonTextBlock_69 {
   widgetClassPath=&CommonTextBlock
   parent=VerticalBox_26
   widget=CommonTextBlock_69
   properties {
    text=Eliminations
    textTransformPolicy=ToUpper
    colorAndOpacity {colorUseRule=UseColor_Specified, specifiedColor=<rgba 0.2122 0.6514 1 1>}
    font {
     bForceMonospaced=false
     bMaterialIsStencil=false
     fontMaterial=null
     fontObject=&BurbankSmall
     letterSpacing=0
     monospacedWidth=1
     size=15
     skewAmount=0
     typefaceFontName=Bold
     outlineSettings {
      bApplyOutlineToDropShadows=false
      bMiteredCorners=false
      bSeparateFillAlpha=false
      outlineMaterial=null
      outlineSize=0
      outlineColor=<rgba 0 0 0 1>
     }
    }
    shadowOffset=<0 0>
   }
   slot {class=VerticalBoxSlot, properties {padding=<ltrb 0 0 0 8>}}
  }
  LocalPlayerElims {
   widgetClassPath=&/BP_AthenaMapPlayerEliminations
   parent=VerticalBox_26
   widget=LocalPlayerElims
   properties {bIsLocalPlayer=true}
   slot {class=VerticalBoxSlot}
  }
  EntryBox_SquadElims {
   widgetClassPath=/Script/FortniteUI.FortDynamicEntryBox
   bInherited=true
   parent=VerticalBox_26
   widget=EntryBox_SquadElims
   properties {
    entryBoxType=Vertical
    entryWidgetClass=&/BP_AthenaMapPlayerEliminations.BP_AthenaMapPlayerEliminations_C
    entrySpacing=<0 8>
   }
   slot {class=VerticalBoxSlot, properties {padding=<ltrb 0 4 0 0>}}
  }
  TutorialHighlight {widgetClassPath=&TutorialHighlight}
 ]
}
