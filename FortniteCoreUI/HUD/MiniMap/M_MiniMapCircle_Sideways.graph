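# Translucent UI material that draws a Sideways zone circle on the minimap, positioned and sized by a center-and-radius vector parameter.
# Offset UVs feed sphere masks that shape the circle, with panning soft noise subtracted to break up its gradient.
# That gradient samples a corruption curve-atlas color, blended with a ring color along a thin band.
# A static switch provides UV rescaling from the minimap parameter collection's scale factor.
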Material /FortniteCoreUI/HUD/MiniMap/M_MiniMapCircle_Sideways
graph {
 properties {
  bUsedWithStaticMesh=false
  blendMode=BLEND_Translucent
  materialDomain=MD_UI
  numCustomizedUVs=1
  pixelDepthOffsetMode=PDOM_Legacy
  usedShadingModels=Default Lit
 }
 nodes [
  n1 = NamedRerouteUsage()
  n2 = NamedRerouteUsage()
  n3 = NamedRerouteUsage()
  n4 = SphereMask(A=@n1, B=@n2, Radius=@n3, Hardness=ScalarParameter() {parameterName=Hardness})
  n5 = OneMinus(None=@n4)
  n6 = ComponentMask(None=Multiply(A=NamedRerouteUsage(), B=ScalarParameter() {defaultValue=64, parameterName=NoiseTiling})) {g=true, r=true}
  n7 = Multiply(
   A=TextureSample(UVs=Panner(Coordinate=@n6) {speedX=0.15, speedY=0.15}) {texture=/Game/Characters/Player/Male/Medium/Bodies/M_MED_Viper/Textures/Glitch/T_SoftNoise}.G
   B=ScalarParameter() {defaultValue=0.25, parameterName=NoiseStrength}
  )
  Clamp(None=Subtract(A=@n5, B=@n7))
  Multiply(A=LinearInterpolate(Alpha=@n5) {constA=0.5, constB=0.9}, B=SphereMask(A=@n1, B=@n2, Radius=@n3, Hardness=Constant() {r=1}))
  n8 = VectorParameter() {
   desc="Circle Origin.X, Origin.Y, Radius (all multiplied by map scale).\r\nA is ignored"
   parameterName=CircleCenterAndRadius
   defaultValue=<rgba 0.5 0.5 1 1>
  }
  NamedRerouteDeclaration(None=Multiply(A=@n8.B, B=ScalarParameter() {defaultValue=0.5, parameterName=RadiusScale})) {name=UI_RadiusInput, nodeColor=<rgba 0 0.4609 1 1>}
  NamedRerouteDeclaration(None=Constant2Vector() {g=0.5, r=0.5}) {name=UI_Circle_Center, nodeColor=<rgba 0 0.5078 1 1>}
  n9 = Subtract(A=TextureCoordinate(), B=ComponentMask(None=@n8.RGB) {g=true, r=true})
  NamedRerouteDeclaration(None=Add(A=Divide(A=@n9)) {constB=0.5}) {name=UI_UV, nodeColor=<rgba 0 0.6719 1 1>}
  n10 = CurveAtlasRowParameter(CurveTime=Clamp(None=Subtract(A=@n4, B=@n7))) {
   atlas=/Game/Athena/Apollo/Environments/Landscape/Materials/Corruption/CorruptionCurveAtlas
   curve=/Game/Athena/Apollo/Environments/Landscape/Materials/Corruption/SidewaysBubbleUIColorCurve
   defaultValue=11
   parameterName=CurveAtlas
  }
  n11 = VectorParameter() {parameterName=RingColor, defaultValue=<rgba 0.6108 0 1 1>}
  n12 = Power(Base=SphereMask(
    A=@n5
    B=Constant() {r=1}
    Radius=ScalarParameter() {defaultValue=0.1, parameterName=RingThickness}
    Hardness=ScalarParameter() {parameterName=RingHardness}
   )) {constExponent=3}
  Power(Base=Multiply(A=LinearInterpolate(A=@n10, B=@n11.RGB, Alpha=@n12)) {constB=3})
  ScalarParameter() {defaultValue=0.5, parameterName=RadiusInput}
  n13 = TextureCoordinate()
  n14 = CollectionParameter() {collection=/Game/Athena/HUD/MiniMap/MPC_Minimap, parameterName=ScaleFactor}
  n15 = OneMinus(None=@n14)
  StaticSwitchParameter(True=Add(A=Multiply(A=@n13, B=@n14) {constB=2}, B=Multiply(A=@n15) {constB=0.5}), False=@n13) {defaultValue=true, parameterName=UseFullScreenScale}
  TextureCoordinate()
 ]
}
