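# Translucent UI master material for NPC icons in the minimap HUD, sampling a texture parameter that defaults to the NPC speech icon.
# A scalar "shear" parameter drives a Rotator's time input to rotate the texture coordinates.
# Only the rotated U is recombined with the original V, so the icon's sampling is skewed horizontally instead of fully rotated.
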Material /FortniteCoreUI/HUD/MiniMap/M_UI_NPC_Master
graph {
 properties {
  bUsedWithStaticMesh=false
  blendMode=BLEND_Translucent
  materialDomain=MD_UI
  pixelDepthOffsetMode=PDOM_Legacy
  usedShadingModels=Default Lit
 }
 nodes [
  n1 = TextureCoordinate()
  n2 = Rotator(Coordinate=@n1, Time=ScalarParameter() {parameterName=shear})
  TextureSampleParameter2D(UVs=AppendVector(A=ComponentMask(None=@n2) {r=true}, B=ComponentMask(None=@n1) {g=true})) {parameterName=Texture, texture=/Game/Athena/HUD/Interaction/T_UI_NPC_SpeechIcon}
 ]
}
