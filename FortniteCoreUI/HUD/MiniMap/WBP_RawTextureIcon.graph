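# Minimap icon widget for FortniteCoreUI that displays a raw texture, built on the native AthenaMapNavigableRawTextureIcon class.
# A SizeBox root holds the inherited Image_Icon, a FortMobileImage with a larger mobile size multiplier and a bottom-center pivot.
# When the OnChallengeIconOverriden event fires, the graph sets the supplied texture as the image brush without matching the texture size.
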WidgetBlueprint /FortniteCoreUI/HUD/MiniMap/WBP_RawTextureIcon
graphs [EventGraph {nodes [
   n1 = Event OnChallengeIconOverriden!() -> (OutputDelegate, then, Icon:Texture 2D Object Reference) {is=title K2Node_Event}
   > SetBrushFromTexture!(
    execute
    self:Image Object Reference=Get Image_Icon() -> (Image_Icon:Image Object Reference)
    Texture:Texture 2D Object Reference=@n1.Icon
    bMatchSize:bool=false
   ) -> (then)
  ]}]
variables [Param_BackingTypeSelector {type=string, default=BackingTypeSelector}]
widgets {
 info {inheritedWidgetCount=1, parentClass=/Script/FortniteUI.AthenaMapNavigableRawTextureIcon, rootWidgetClass=&SizeBox, widgetCount=2}
 widgets [
  SizeBox_Icon {widgetClassPath=&SizeBox, bIsVariable=true, widget=SizeBox_Icon}
  Image_Icon {
   widgetClassPath=/Script/FortniteUI.FortMobileImage
   bInherited=true
   bIsVariable=true
   parent=SizeBox_Icon
   widget=Image_Icon
   properties {
    mobileSizeMultiplier=1.5
    visibility=SelfHitTestInvisible
    brush {
     bIsDynamicallyLoaded=false
     drawAs=Image
     imageType=FullColor
     mirroring=NoMirror
     resourceName=None
     resourceObject=null
     tiling=NoTile
     imageSize=<48 48>
     margin=<ltrb 0 0 0 0>
     outlineSettings {
      bUseBrushTransparency=false
      roundingType=HalfHeightRadius
      width=0
      color {colorUseRule=UseColor_Specified, specifiedColor=<rgba 0 0 0 0>}
      cornerRadii=<0 0 0 0>
     }
     tintColor {colorUseRule=UseColor_Specified, specifiedColor=<rgba 1 1 1 1>}
     uVRegion {bIsValid=false, max=<0 0>, min=<0 0>}
    }
    renderTransformPivot=<0.5 1>
   }
   slot {class=SizeBoxSlot, properties {horizontalAlignment=HAlign_Left, verticalAlignment=VAlign_Top}}
  }
 ]
}
