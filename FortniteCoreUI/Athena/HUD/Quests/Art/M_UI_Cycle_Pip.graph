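# Translucent UI material for a quest HUD cycle pip, drawing a centre dot inside a ring that shows radial progress.
# One exponential radial gradient is thresholded into the dot and, minus a cutout, the ring, with parameterised sizes and edges.
# A rotated UV-to-angle sweep, cut by Progress, fills the ring over its dim, opacity-clamped track.
# Dot and ring masks get separate alpha multipliers and are summed, alongside a constant white.
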Material /FortniteCoreUI/Athena/HUD/Quests/Art/M_UI_Cycle_Pip
graph {
 properties {
  bUsedWithStaticMesh=false
  blendMode=BLEND_Translucent
  materialDomain=MD_UI
  pixelDepthOffsetMode=PDOM_Legacy
  usedShadingModels=Default Lit
 }
 nodes [
  n1 = MaterialFunctionCall(Density=ScalarParameter() {defaultValue=2, parameterName=Radius}) {materialFunction=&RadialGradientExponential}
  n2 = Multiply(
   A=Subtract(A=@n1.RadialGradientExponential, B=ScalarParameter() {defaultValue=0.7, parameterName=CentreCircle-size})
   B=ScalarParameter() {defaultValue=8, parameterName=CentreCircle-Edge}
  )
  n3 = Subtract(A=@n1.RadialGradientExponential, B=ScalarParameter() {parameterName=OutlineCircle-size})
  n4 = Multiply(A=@n3, B=ScalarParameter() {defaultValue=6, parameterName=OutlineCircle-Edge})
  n5 = Multiply(
   A=Subtract(A=@n3, B=ScalarParameter() {defaultValue=0.15, parameterName=OutlineCircle-Cutout-Size})
   B=ScalarParameter() {defaultValue=4, parameterName=OutlineCircle-Cutout-Edge}
  )
  n6 = Saturate(None=Subtract(A=Saturate(None=@n4), B=Saturate(None=@n5)))
  n7 = Clamp(None=@n6, Max=ScalarParameter() {defaultValue=0.3, parameterName=OuterCircle-Opacity})
  n8 = Rotator(
   Coordinate=Subtract(A=TextureCoordinate()) {constB=0.5}
   Time=ScalarParameter() {defaultValue=0.25, group=ProgressMeter, parameterName=VectorRad-Rotate}
  ) {centerX=0, centerY=0, speed=6.2832}
  n9 = Subtract(
   A=OneMinus(None=MaterialFunctionCall(Vector or UVs=@n8) {materialFunction=&VectorToRadialValue}.Vector Converted to Angle)
   B=LinearInterpolate(Alpha=ScalarParameter() {defaultValue=0.25, group=ProgressMeter, parameterName=Progress}) {constA=1, constB=0}
  )
  n10 = ScalarParameter() {defaultValue=128, group=ProgressMeter, parameterName=Progress-Edge}
  n11 = Multiply(A=@n6, B=Saturate(None=Multiply(A=@n9, B=@n10)))
  Add(
   A=Multiply(A=Saturate(None=@n2), B=ScalarParameter() {defaultValue=0.5, parameterName=CentreCircle-Alpha})
   B=Multiply(A=Add(A=@n7, B=@n11), B=ScalarParameter() {parameterName=OuterCircle-GlobalAlpha})
  )
  Constant4Vector() {constant=<rgba 1 1 1 1>}
 ]
}
