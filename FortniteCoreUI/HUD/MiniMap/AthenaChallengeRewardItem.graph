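# Widget Blueprint for a single challenge reward entry in the Athena HUD, built on the native AthenaChallengeRewardItem class.
# It stacks a lazy-loaded reward icon over a quest backing, with a slanted red lock badge at lower right and a count label below.
# The parent's inherited widgets supply the main elements.
# The only graph logic, BP_OnSetSelected, tints the count text dark navy when selected and white otherwise.
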WidgetBlueprint /FortniteCoreUI/HUD/MiniMap/AthenaChallengeRewardItem
graphs [EventGraph {nodes [
   n1 = Event BP_OnSetSelected!() -> (OutputDelegate, then, bIsSelected:bool) {is=title K2Node_Event}
   > SetColorAndOpacity(
    self=Get Text_RewardCount() -> (Text_RewardCount:Common Text Object Reference)
    InColorAndOpacity=Make Slate Color(SpecifiedColor=SelectColor(A=(R=0,G=0.0056,B=0.1329,A=1), B=(R=1,G=1,B=1,A=1), bPickA=@n1.bIsSelected))
   )
  ]}]
variables []
widgets {
 info {inheritedWidgetCount=4, parentClass=/Script/FortniteUI.AthenaChallengeRewardItem, rootWidgetClass=&VerticalBox, widgetCount=9}
 widgets [
  VerticalBox_88 {widgetClassPath=&VerticalBox, widget=VerticalBox_88}
  Overlay_61 {
   widgetClassPath=&Overlay
   parent=VerticalBox_88
   widget=Overlay_61
   slot {class=VerticalBoxSlot, properties {horizontalAlignment=HAlign_Center, verticalAlignment=VAlign_Center}}
  }
  Image_RewardBacking {
   widgetClassPath=&Image
   bInherited=true
   bIsVariable=true
   parent=Overlay_61
   widget=Image_RewardBacking
   properties {brush {
     bIsDynamicallyLoaded=false
     drawAs=Image
     imageType=NoImage
     mirroring=NoMirror
     resourceName=None
     resourceObject=/FortniteCoreUI/Athena/HUD/Quests/Art/M_UI_Quest_Cosmetic-Backing
     tiling=NoTile
     imageSize=<64 64>
     margin=<ltrb 0 0 0 0>
     outlineSettings {
      bUseBrushTransparency=false
      roundingType=HalfHeightRadius
      width=0
      color {colorUseRule=UseColor_Specified, specifiedColor=<rgba 0 0 0 0>}
      cornerRadii=<0 0 0 0>
     }
     tintColor {colorUseRule=UseColor_Specified, specifiedColor=<rgba 1 1 1 1>}
     uVRegion {bIsValid=false, max=<0 0>, min=<0 0>}
    }}
   slot {class=OverlaySlot, properties {horizontalAlignment=HAlign_Center}}
  }
  SizeBox_47 {
   widgetClassPath=&SizeBox
   parent=Overlay_61
   widget=SizeBox_47
   properties {bOverride_HeightOverride=true, bOverride_WidthOverride=true, heightOverride=64, widthOverride=64}
   slot {class=OverlaySlot, properties {padding=<ltrb 0 0 0 4>}}
  }
  LazyImage_RewardIcon {
   widgetClassPath=&FortLazyImage
   bInherited=true
   bIsVariable=true
   parent=SizeBox_47
   widget=LazyImage_RewardIcon
   properties {brush {
     bIsDynamicallyLoaded=false
     drawAs=Image
     imageType=FullColor
     mirroring=NoMirror
     resourceName=None
     resourceObject=/Game/UI/Foundation/Textures/Icons/Items/UniversalXP/T_UI_UniversalXP_Icon_Standard_64
     tiling=NoTile
     imageSize=<64 64>
     margin=<ltrb 0 0 0 0>
     outlineSettings {
      bUseBrushTransparency=false
      roundingType=HalfHeightRadius
      width=0
      color {colorUseRule=UseColor_Specified, specifiedColor=<rgba 0 0 0 0>}
      cornerRadii=<0 0 0 0>
     }
     tintColor {colorUseRule=UseColor_Specified, specifiedColor=<rgba 1 1 1 1>}
     uVRegion {bIsValid=false, max=<0 0>, min=<0 0>}
    }}
   slot {class=SizeBoxSlot}
  }
  Panel_Lock {
   widgetClassPath=&Overlay
   bInherited=true
   parent=Overlay_61
   widget=Panel_Lock
   properties {renderTransform {angle=0, scale=<1 1>, shear=<0 0>, translation=<15 0>}, renderTransformPivot=<0 1>}
   slot {class=OverlaySlot, properties {horizontalAlignment=HAlign_Right, verticalAlignment=VAlign_Bottom}}
  }
  Image_305 {
   widgetClassPath=&Image
   bIsVariable=true
   parent=Panel_Lock
   widget=Image_305
   properties {
    brush {
     bIsDynamicallyLoaded=false
     drawAs=Image
     imageType=NoImage
     mirroring=NoMirror
     resourceName=None
     resourceObject=null
     tiling=NoTile
     imageSize=<22 22>
     margin=<ltrb 0 0 0 0>
     outlineSettings {
      bUseBrushTransparency=false
      roundingType=HalfHeightRadius
      width=0
      color {colorUseRule=UseColor_Specified, specifiedColor=<rgba 0 0 0 0>}
      cornerRadii=<0 0 0 0>
     }
     tintColor {colorUseRule=UseColor_Specified, specifiedColor=<rgba 1 1 1 1>}
     uVRegion {bIsValid=false, max=<0 0>, min=<0 0>}
    }
    colorAndOpacity=<rgba 1 0 0.2988 1>
    renderTransform {angle=0, scale=<1 1>, shear=<-9 0>, translation=<0 0>}
   }
   slot {class=OverlaySlot, properties {horizontalAlignment=HAlign_Fill, verticalAlignment=VAlign_Fill}}
  }
  ImgLockIcon {
   widgetClassPath=&Image
   parent=Panel_Lock
   widget=ImgLockIcon
   properties {brush {
     bIsDynamicallyLoaded=false
     drawAs=Image
     imageType=NoImage
     mirroring=NoMirror
     resourceName=None
     resourceObject=/Game/UI/Foundation/Textures/Icons/Locks/T-Icon-Lock-64
     tiling=NoTile
     imageSize=<16 16>
     margin=<ltrb 0 0 0 0>
     outlineSettings {
      bUseBrushTransparency=false
      roundingType=HalfHeightRadius
      width=0
      color {colorUseRule=UseColor_Specified, specifiedColor=<rgba 0 0 0 0>}
      cornerRadii=<0 0 0 0>
     }
     tintColor {colorUseRule=UseColor_Specified, specifiedColor=<rgba 1 1 1 1>}
     uVRegion {bIsValid=false, max=<0 0>, min=<0 0>}
    }}
   slot {class=OverlaySlot, properties {horizontalAlignment=HAlign_Center, verticalAlignment=VAlign_Center}}
  }
  Text_RewardCount {
   widgetClassPath=&CommonTextBlock
   bInherited=true
   bIsVariable=true
   parent=VerticalBox_88
   widget=Text_RewardCount
   properties {
    text=100K
    font {
     bForceMonospaced=false
     bMaterialIsStencil=false
     fontMaterial=null
     fontObject=&BurbankSmall
     letterSpacing=0
     monospacedWidth=1
     size=15
     skewAmount=0
     typefaceFontName=Bold
     outlineSettings {
      bApplyOutlineToDropShadows=false
      bMiteredCorners=false
      bSeparateFillAlpha=false
      outlineMaterial=null
      outlineSize=0
      outlineColor=<rgba 0 0 0 1>
     }
    }
    shadowOffset=<0 0>
   }
   slot {class=VerticalBoxSlot, properties {horizontalAlignment=HAlign_Center, padding=<ltrb 0 -16 0 4>}}
  }
 ]
}
