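# Small circular pip widget for AthenaChallengeRewardsMultiMarker, drawn by a single image using the M_UI_Cycle_Pip material.
# Selection changes tint the image dark navy when selected and white otherwise.
# Active-state changes stop all animations, then play Active, which fills the material's Progress over three seconds at full alpha, or Inactive, which hides the outer circle and dims the centre.
# Both animations drive material parameters directly through widget material tracks.
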WidgetBlueprint /FortniteCoreUI/HUD/MiniMap/AthenaChallengeRewardsMultiMarker
animations [
 Active {
  binding_count=1
  length_seconds=3
  object=&self:Active
  sequence {bindings [Image_Circle {
     id=#1
     possessed_class=Image
     tracks [{
       type=MovieSceneWidgetMaterialTrack
       sections [{
         type=MovieSceneParameterSection
         end_frame=60
         start_frame=0
         channels [
          OuterCircle-GlobalAlpha {type=MovieSceneScriptingFloatChannel, const [1,1]}
          CentreCircle-Alpha {type=MovieSceneScriptingFloatChannel, const [1,1]}
          Progress {type=MovieSceneScriptingFloatChannel, keys [[0,0],[60,1]]}
         ]
        }]
      }]
    }]}
 }
 Inactive {
  binding_count=1
  length_seconds=0.25
  object=&self:Inactive
  sequence {bindings [Image_Circle {
     id=#2
     possessed_class=Image
     tracks [{
       type=MovieSceneWidgetMaterialTrack
       sections [{
         type=MovieSceneParameterSection
         end_frame=5
         start_frame=0
         channels [
          CentreCircle-Alpha {type=MovieSceneScriptingFloatChannel, const [0.5,1]}
          OuterCircle-GlobalAlpha {type=MovieSceneScriptingFloatChannel, const [0,1]}
          Progress {type=MovieSceneScriptingFloatChannel, const [0,1]}
         ]
        }]
      }]
    }]}
 }
]
graphs [EventGraph {nodes [
   n1 = Event BP_OnSetSelected!() -> (OutputDelegate, then, bIsSelected:bool) {is=title K2Node_Event}
   > SetColorAndOpacity(
    self:Image Object Reference=Get Image_Circle() -> (Image_Circle:Image Object Reference)
    InColorAndOpacity:Linear Color Structure=Select(Option 0:Linear Color Structure=(R=1,G=1,B=1,A=1), Option 1:Linear Color Structure=(R=0,G=0.0056,B=0.1329,A=1), Index=@n1.bIsSelected) -> (ReturnValue:Linear Color Structure)
   )
   n2 = Event BP_OnSetMarkerActive!() -> (OutputDelegate, then, bIsActive:bool) {is=title K2Node_Event}
   > StopAllAnimations(self:User Widget Object Reference)
   > n3 = Branch(Condition=@n2.bIsActive)
   > PlayAnimation(InAnimation=Get Active() -> (Active:Widget Animation Object Reference))
   PlayAnimation(execute=@n3.else, InAnimation=Get Inactive() -> (Inactive:Widget Animation Object Reference))
  ]}]
variables []
widgets {
 info {inheritedWidgetCount=0, parentClass=/Script/FortniteUI.AthenaChallengeRewardsMultiMarker, rootWidgetClass=&Image, widgetCount=1}
 widgets [Image_Circle {
   widgetClassPath=&Image
   bIsVariable=true
   widget=Image_Circle
   properties {brush {
     bIsDynamicallyLoaded=false
     drawAs=Image
     imageType=NoImage
     mirroring=NoMirror
     resourceName=None
     resourceObject=/FortniteCoreUI/Athena/HUD/Quests/Art/M_UI_Cycle_Pip
     tiling=NoTile
     imageSize=<11 11>
     margin=<ltrb 0 0 0 0>
     outlineSettings {
      bUseBrushTransparency=false
      roundingType=HalfHeightRadius
      width=0
      color {colorUseRule=UseColor_Specified, specifiedColor=<rgba 0 0 0 0>}
      cornerRadii=<0 0 0 0>
     }
     tintColor {colorUseRule=UseColor_Specified, specifiedColor=<rgba 1 1 1 1>}
     uVRegion {bIsValid=false, max=<0 0>, min=<0 0>}
    }}
  }]
}
