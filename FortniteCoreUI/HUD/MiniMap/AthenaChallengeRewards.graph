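# AthenaChallengeRewards is a minimap HUD widget blueprint, subclassing the native AthenaChallengeRewards class, for showing challenge rewards.
# A vertical box stacks an overlay holding one reward item widget above a centered dynamic entry box that creates AthenaChallengeRewardsMultiMarker entries with slightly overlapping spacing.
# The event graph is empty, so the file supplies only layout, with the container and marker box inherited from the parent class.
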WidgetBlueprint /FortniteCoreUI/HUD/MiniMap/AthenaChallengeRewards
graphs [EventGraph {links [], nodes []}]
variables []
widgets {
 info {inheritedWidgetCount=2, parentClass=/Script/FortniteUI.AthenaChallengeRewards, rootWidgetClass=&VerticalBox, widgetCount=4}
 widgets [
  VerticalBox_88 {widgetClassPath=&VerticalBox, widget=VerticalBox_88}
  ItemContainer {widgetClassPath=&Overlay, bInherited=true, parent=VerticalBox_88, widget=ItemContainer, slot {class=VerticalBoxSlot}}
  AthenaChallengeRewardItem {
   widgetClassPath=&/AthenaChallengeRewardItem
   bIsVariable=true
   parent=ItemContainer
   widget=AthenaChallengeRewardItem
   slot {class=OverlaySlot}
  }
  DynamicBox_MultiMarkers {
   widgetClassPath=/Script/FortniteUI.FortDynamicEntryBox
   bInherited=true
   bIsVariable=true
   parent=VerticalBox_88
   widget=DynamicBox_MultiMarkers
   properties {entryWidgetClass=&/AthenaChallengeRewardsMultiMarker.AthenaChallengeRewardsMultiMarker_C, entrySpacing=<-4 0>}
   slot {class=VerticalBoxSlot, properties {horizontalAlignment=HAlign_Center}}
  }
 ]
}
